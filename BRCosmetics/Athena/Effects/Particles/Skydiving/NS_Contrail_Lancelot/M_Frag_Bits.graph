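# Lit material for the mesh fragment particles in the Lancelot skydiving contrail, used with Niagara.
# Vertex color G switches between gold and standard diffuse, normal and OARM sets, while R swaps in a third diffuse.
# A view-angle curve-atlas lookup tints masked areas and a dimmed copy feeds emissive adjustments.
# A UV1 vector LUT and WiggleSpeed dynamic parameter build a sine wiggle offset, varied by vertex color and gated by UseWPO?.
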Material /BRCosmetics/Athena/Effects/Particles/Skydiving/NS_Contrail_Lancelot/M_Frag_Bits
graph {
 properties {
  bUsedWithNiagaraMeshParticles=true
  pixelDepthOffsetMode=PDOM_Legacy
  previewMesh=&/SM_Fragment_Bits
  usedShadingModels=Default Lit
 }
 nodes [
  Constant() {r=1}
  n1 = VertexColor()
  n2 = Constant() {r=1}
  n3 = If(A=@n1.G, B=Constant() {r=1}, "A > B"=@n2, "A == B"=@n2, "A < B"=Constant())
  n4 = Constant()
  n5 = LinearInterpolate(
   A=LinearInterpolate(
    A=TextureSample() {texture=&/T_Frag_Diffuse_Gold_Full}.RGB
    B=TextureSample() {texture=&/T_Frag_Diffuse}.RGB
    Alpha=@n3
   )
   B=TextureSample() {texture=&/T_Frag_Diffuse_S}.RGB
   Alpha=If(A=@n1.R, B=Constant() {r=0.5}, "A > B"=@n4, "A == B"=@n4, "A < B"=Constant() {r=1})
  )
  n6 = CameraVectorWS()
  n7 = PixelNormalWS()
  n8 = CurveAtlasRowParameter(CurveTime=Clamp(None=DotProduct(A=@n6, B=@n7))) {atlas=&AthenaCurveAtlas01, curve=/Game/Athena/LinearColorCurves/Curve_IcoQueenColor_Gold, defaultValue=55}
  n9 = ChannelMaskParameter(None=@n8) {group=LUT, parameterName=LUTChannel}
  n10 = TextureSample() {texture=&/LP_Frag_lambert1_MaskMap_02}
  n11 = TextureSample() {texture=&/LP_Frag_lambert1_MaskMap}
  n12 = Multiply(A=LinearInterpolate(A=@n5, B=Multiply(A=@n9.R, B=@n5), Alpha=LinearInterpolate(A=@n10.G, B=@n11.G, Alpha=@n3)) {constAlpha=0}) {constB=0.1}
  MaterialFunctionCall(Emissive=@n12) {materialFunction=&MF_FX_EmissiveAdjustments}
  Reroute(None=@n12)
  n13 = DynamicParameter() {defaultValue=<rgba 1.618 1 1 1>, paramNames [WiggleSpeed, Time, "", ""]}
  n14 = TextureSample(UVs=TextureCoordinate() {coordinateIndex=1}) {
   samplerType=SAMPLERTYPE_LinearColor
   texture=&/T_Contrail_Lancelot_VectorLUT
  }
  n15 = Add(A=Multiply(A=@n13.WiggleSpeed, B=@n13.WiggleSpeed), B=@n14.A)
  n16 = VertexColor()
  n17 = Constant() {r=1}
  n18 = If(A=@n16.G, B=Constant() {r=1}, "A > B"=@n17, "A == B"=@n17, "A < B"=Constant() {r=0.7})
  n19 = LinearInterpolate(A=@n15, B=@n15, Alpha=@n18) {constB=0}
  n20 = Multiply(A=@n19, B=@n18) {constB=0}
  n21 = Constant()
  n22 = If(A=@n16.R, B=Constant() {r=0.5}, "A > B"=@n21, "A == B"=@n21, "A < B"=Constant() {r=1})
  n23 = ConstantBiasScale(None=Sine(None=LinearInterpolate(A=@n19, B=@n20, Alpha=@n22) {constB=0})) {desc="1, 0.5", scale=0.75}
  n24 = Transform(None=ConstantBiasScale(None=@n14.RGB) {bias=-0.5, desc="-0.5, 2", scale=2}) {transformSourceType=TRANSFORMSOURCE_Local}
  Reroute(None=LinearInterpolate(B=Multiply(A=@n23, B=@n24), Alpha=ScalarParameter() {defaultValue=1, parameterName=UseWPO?}) {constB=0})
  n25 = TextureSample() {samplerType=SAMPLERTYPE_Normal, texture=&/T_Frag_Normal_Gold_Full}
  n26 = TextureSample() {samplerType=SAMPLERTYPE_Normal, texture=&/T_Frag_Normal}
  Reroute(None=Normalize(VectorInput=LinearInterpolate(A=@n25.RGB, B=@n26.RGB, Alpha=@n3)))
  n27 = TextureSample() {texture=&/T_Frag_OARM_Gold}
  n28 = TextureSample() {texture=&/T_Frag_OARM}
  n29 = Multiply(A=LinearInterpolate(A=@n27.G, B=@n28.G, Alpha=@n3))
  Reroute(None=Saturate(None=Add(A=@n29) {constB=0.2}))
  n30 = ScalarParameter() {defaultValue=0.5, group=Masks, parameterName=MetalMaskPower}
  n31 = LinearInterpolate(A=@n27.B, B=@n28.B, Alpha=@n3)
  Reroute(None=Saturate(None=Power(Base=@n31, Exp=@n30)))
  TextureSample() {texture=&/LP_Frag_lambert1_Specular}
  Time()
 ]
}
