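# Unlit alpha-composite material for Niagara ribbons, forming the Lancelot skydiving contrail.
# A panning trail texture, faded along U, multiplies an edge-tapered gradient broken up by skewed panning noise and erosion borrowed from the DarkAndLight contrail.
# A depth fade with distance varying along the ribbon also fades the start.
# The resulting mask is tinted by a variant color for emissive and scaled by particle alpha, while a dynamic parameter drives panning time.
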Material /BRCosmetics/Athena/Effects/Particles/Skydiving/NS_Contrail_Lancelot/M_Lancelot_Trail
graph {
 properties {
  bUsedWithNiagaraRibbons=true
  blendMode=BLEND_AlphaComposite
  pixelDepthOffsetMode=PDOM_Legacy
  shadingModel=MSM_Unlit
  usedShadingModels=Unlit
 }
 nodes [
  n1 = Divide(A=MaterialFunctionCall() {materialFunction=/Game/Effects/Fort_Effects/MaterialFunctions/MF_U}.U) {constB=0.1}
  n2 = ScalarParameter() {defaultValue=222, parameterName=DepthFadeDistanceA}
  n3 = ScalarParameter() {defaultValue=22, parameterName=DepthFadeDistanceB}
  n4 = Saturate(None=Subtract(A=@n1) {constB=0.25})
  n5 = MaterialFunctionCall(Float2=TextureCoordinate()) {materialFunction=&BreakOutFloat2Components}
  n6 = ScalarParameter() {defaultValue=1, parameterName=FullUV_UScale}
  n7 = ScalarParameter() {defaultValue=1, parameterName=FullUV_VScale}
  n8 = LinearInterpolate(
   A=TextureCoordinate() {uTiling=0.7, vTiling=0.8}
   B=AppendVector(A=Multiply(A=@n5.R, B=@n6), B=Multiply(A=@n5.G, B=@n7))
   Alpha=ScalarParameter() {parameterName=UseFullUV_ForNoise}
  )
  n9 = ComponentMask(None=@n8) {r=true}
  n10 = DynamicParameter() {paramNames [TimeOffset, Time, "", ""]}
  n11 = Add(A=@n10.Time, B=@n10.Time)
  n12 = ScalarParameter() {defaultValue=-0.5, parameterName=Panner Speed X 2}
  n13 = ScalarParameter() {parameterName=Panner Speed Y 2}
  n14 = TextureSample(UVs=Panner(Coordinate=@n8, Time=@n11, Speed=AppendVector(A=@n12, B=@n13)) {speedX=-0.5}) {texture=&/T_Lancelot_Trail}
  n15 = Power(Base=Multiply(A=OneMinus(None=@n9), B=Add(A=@n14.R, B=@n14.R)), Exp=ScalarParameter() {defaultValue=1, parameterName=EmissiveNoisePower})
  n16 = ScalarParameter() {defaultValue=1, parameterName=EmissiveNoiseMult}
  n17 = TextureCoordinate()
  n18 = ComponentMask(None=@n17) {r=true}
  n19 = Multiply(A=@n18, B=@n18)
  n20 = ComponentMask(None=TextureCoordinate()) {g=true}
  n21 = Constant() {r=0.5}
  n22 = Abs(None=Divide(A=Subtract(A=@n20, B=@n21), B=@n21))
  n23 = ScalarParameter() {defaultValue=-1, parameterName=Noise Skew, sliderMax=1}
  n24 = AppendVector(A=Add(A=ComponentMask(None=@n17) {g=true}, B=Multiply(A=@n18, B=@n23)), B=@n18)
  n25 = AppendVector(A=ScalarParameter() {defaultValue=0.3, parameterName=Noise Scale U}, B=ScalarParameter() {defaultValue=1.5, parameterName=Noise Scale V})
  n26 = ScalarParameter() {defaultValue=-0.04, parameterName=Panner Speed X}
  n27 = ScalarParameter() {defaultValue=-1.66, parameterName=Panner Speed Y}
  n28 = TextureSample(UVs=Panner(Coordinate=Multiply(A=@n24, B=@n25), Time=@n11, Speed=AppendVector(A=@n26, B=@n27)) {bFractionalPart=true}) {
   samplerType=SAMPLERTYPE_LinearGrayscale
   texture=/BRCosmetics/Athena/Effects/Particles/Skydiving/NS_Contrail_DarkAndLight/T_DarkAndLight_Noise
  }
  n29 = MaterialFunctionCall!(
   Gradient=Multiply(A=OneMinus(None=@n19), B=OneMinus(None=@n22))
   Noise=OneMinus(None=@n28.R)
   Modulation Amount=Multiply(A=SquareRoot(None=@n18), B=ScalarParameter() {defaultValue=2, parameterName=Noise Modulation, sliderMax=9})
  ) {materialFunction=/Game/Effects/Fort_Effects/MaterialFunctions/MF_GradientNoiseModulate}
  n30 = OneMinus(None=MaterialFunctionCall!(Gradient=OneMinus(None=@n29.Result), Erosion Amount=ScalarParameter() {defaultValue=0.5, parameterName=Noise Erosion, sliderMax=1}) {materialFunction=/BRCosmetics/Athena/Effects/Particles/Skydiving/NS_Contrail_DarkAndLight/Functions/MF_ErosionNormalized}.Result)
  n31 = Multiply(
   A=DepthFade(Opacity=Saturate(None=@n1), FadeDistance=LinearInterpolate(A=@n2, B=@n3, Alpha=@n4))
   B=Multiply(A=Multiply(A=@n15, B=@n16), B=Saturate(None=@n30))
  )
  MaterialFunctionCall(Emissive=Multiply(A=VectorParameter() {parameterName=Variant_Color_M, defaultValue=<rgba 0 1 1 1>}.RGB, B=@n31)) {materialFunction=&MF_FX_EmissiveAdjustments}
  Multiply(A=ParticleColor().A, B=@n31)
  Time()
 ]
}
