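# Translucent unlit rune material for Niagara sprites, ribbons and meshes in the Lancelot skydiving contrail effect.
# Particle color times the rune mask texture feeds an emissive adjustment function.
# Particle alpha times the same mask forms opacity, depth-faded on the default shading path but left unfaded on forward and mobile.
# A RuneColorDim color blend and a constant color feed no other node in the graph.
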Material /BRCosmetics/Athena/Effects/Particles/Skydiving/NS_Contrail_Lancelot/M_Runes_Lancelot
graph {
 properties {
  bUsedWithNiagaraMeshParticles=true
  bUsedWithNiagaraRibbons=true
  bUsedWithNiagaraSprites=true
  blendMode=BLEND_Translucent
  pixelDepthOffsetMode=PDOM_Legacy
  shadingModel=MSM_Unlit
  usedShadingModels=Unlit
 }
 nodes [
  Constant3Vector() {constant=<rgba 0 2.5 5 0>}
  n1 = ParticleColor()
  n2 = TextureSample() {samplerType=SAMPLERTYPE_Masks, texture=&/T_Runes_Lancelot}
  LinearInterpolate(A=@n1.RGB, B=VectorParameter() {parameterName=RuneColorDim, defaultValue=<rgba 1 1 1 0>}.RGB, Alpha=@n2.RGB)
  MaterialFunctionCall(Emissive=Multiply(A=@n1.RGB, B=@n2.RGB)) {materialFunction=&MF_FX_EmissiveAdjustments}
  n3 = Multiply(A=@n1.A, B=@n2.RGB)
  ShadingPathSwitch(Default=DepthFade(Opacity=@n3, FadeDistance=ScalarParameter() {parameterName=Fade Distance}), Forward=@n3, Mobile=@n3)
 ]
}
