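# Niagara skydiving contrail for the Lancelot cosmetic, trailing flames and mesh bits from socket or bone locations on two user-supplied skeletal meshes.
# Sprite flame emitters take the variant color, with mirrored setups per mesh, while mesh emitters add falling and floating bits.
# Fire meshes swing on spring-driven pendulums, passing pivot and position to their materials through dynamic parameters.
# System.Dive blends velocities, drag and offsets, with separate front-end preview values.
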NiagaraSystem /BRCosmetics/Athena/Effects/Particles/Skydiving/NS_Contrail_Lancelot/NS_Contrail_Lancelot
dependencies {
 usedDataInterfaces [
  &NiagaraDataInterfaceColorCurve
  &NiagaraDataInterfaceCurve
  &NiagaraDataInterfaceVector2DCurve
  &NiagaraDataInterfaceArrayFloat3
  &NiagaraDataInterfaceVectorCurve
  &NiagaraDataInterfaceArrayFloat
 ]
 usedDynamicInputs [&SetBoolByFloatComparison]
 usedModules [
  &self:SystemSpawnScript.SystemScriptSource.SystemScriptGraph.NiagaraNodeAssignment_2.SetVariables_#1
  &self:SystemSpawnScript.SystemScriptSource.SystemScriptGraph.NiagaraNodeAssignment_0.SetVariables_#2
  &SystemState
  &self:SystemSpawnScript.SystemScriptSource.SystemScriptGraph.NiagaraNodeAssignment_1.SetVariables_#3
  &self:SystemSpawnScript.SystemScriptSource.SystemScriptGraph.NiagaraNodeAssignment_4.SetVariables_#4
 ]
}
emitters {
 Bits {
  emitterUpdateScript [
   &EmitterState()
   &SpawnRate(SpawnRate=15, Use Spawn Probability=false)
  ]
  particleSpawnScript [
   &FN_SetSpawnGroup(Number of Spawn Groups=2)
   /SetVariables(Particles.MeshIndex=&RandomRangeInt(Minimum=0, Maximum=2)) {isSetParameters=true}
   &InitializeParticle(
    Lifetime=&RandomRangeFloat(Minimum=0.5, Maximum=1)
    Mesh Scale Mode=::NewEnumerator1
    Mesh Uniform Scale Min=0.5
    UsePositionOffset=false
    Write Material Random=true
   )
   &SkeletalMeshLocation(
    Bone / Socket Index=$Particles.SpawnGroup
    Bone Sampling Mode=::NewEnumerator4
    Inherit Velocity=false
    MaskSpawnGroup=false
    Skeletal Mesh=$User.Skeletal Mesh02
   )
   &CameraOffset(Camera Offset Amount=&RandomRangeFloat(Minimum=10, Maximum=20))
   /Niagara/Modules/Spawn/Location/SphereLocation(MaskSpawnGroup=false, Module.Override Randomness=false, Sphere Radius=20, Surface Only=false)
   &AddVelocity(
    Coordinate Space=::World
    Velocity=&Lerp_Vector(
     A=<0 0 400>
     B=&FN_FrontEndVector3Modifier(Gameplay=<0 0 1600>, Front End=<0 0 0>)
     Alpha=$System.Dive
    )
   )
   &AddVelocity(Velocity=&FN_FrontEndVector3Modifier(
     Gameplay=&Lerp_Vector(A=<0 75 0>, B=<1 200 1>, Alpha=$System.Dive)
     Front End=&Lerp_Vector(A=<0 -10 0>, B=<1 -400 600>, Alpha=$System.Dive)
    )) {module=AddVelocity003}
   &InitialMeshOrientation(Use Rotation Vector=false)
  ]
  particleUpdateScript [
   &InheritSourceMovement(Applied Position Scale=<0.5 0.5 1>, ApplyForce=false, ApplyPosition=true, ApplyPositionFalloff=false)
   &ParticleState()
   &Drag(Drag=&Lerp_Float(A=1, B=2, Alpha=$System.Dive), Use Linear Drag=true, Use Rotational Drag=false)
   &SolveForcesAndVelocity(Clamp Velocity=false, Limit Acceleration=false)
   &UpdateMeshOrientation(Rotation Coordinate Space=::NewEnumerator1, Rotation Rate=0.1, Rotation Vector=<1 1 1>)
   &ScaleMeshSize(Scale Factor=&VectorFromCurve(VectorCurve=DataInterface() {propertyValues {
       XCurve {keys [<at 0 1 RCIM_Cubic>, {arriveTangent=-3.0986, interpMode=RCIM_Cubic, leaveTangent=-3.0986, tangentMode=RCTM_User, time=1, value=0}]}
       YCurve {keys [<at 0 1 RCIM_Cubic>, {arriveTangent=-3.0986, interpMode=RCIM_Cubic, leaveTangent=-3.0986, tangentMode=RCTM_User, time=1, value=0}]}
       ZCurve {keys [<at 0 1 RCIM_Cubic>, {arriveTangent=-3.0986, interpMode=RCIM_Cubic, leaveTangent=-3.0986, tangentMode=RCTM_User, time=1, value=0}]}
      }}))
  ]
  rendererClasses [&NiagaraMeshRendererProperties]
  renderers [{rendererClass=&NiagaraMeshRendererProperties, rendererIndex=0}]
 }
 Bottom_Flame {
  emitterUpdateScript [
   &EmitterState()
   &SpawnRate(SpawnRate=12, Use Spawn Probability=false)
  ]
  particleSpawnScript [
   /Niagara/Modules/Spawn/Initialization/InitializeParticle(
    Color=$User.Variant_Color
    Lifetime=&UniformRangedFloat(Minimum=0.6, Maximum=0.7)
    Module.Write Color=true
    Module.Write Lifetime=true
    Module.Write Mass=false
    Module.Write Material Random=true
    Module.Write Position=false
    Module.Write Scale=false
    Module.Write SpriteRotation=true
    Module.Write SpriteSize=true
    Position=&SimulationPosition()
    Sprite Rotation=&UniformRangedFloat(Minimum=-20, Maximum=20)
    Sprite Size=&MakeCustomVector2DFromBool(Bool=&RandomBool(), True Vector2D=<50 50>, False Vector2D=<-50 50>)
   )
   &FN_SetSpawnGroup(Number of Spawn Groups=2)
   &SkeletalMeshLocation(
    Bone / Socket Index=$Particles.SpawnGroup
    Bone Sampling Mode=::NewEnumerator4
    Inherit Velocity=false
    MaskSpawnGroup=false
    Skeletal Mesh=$User.SkeletalMesh
   )
   &SphereLocation(MaskSpawnGroup=false, Module.Override Randomness=false, Non Uniform Scale=<1 1 0.1>, Sphere Radius=5, Surface Only=false)
   &DynamicMaterialParameters(
    Index 0 Param 2=&UniformRangedFloat(Minimum=0, Maximum=1)
    Index 0 Param 3=1
    Index 0 Param 4=&UniformRangedFloat(Minimum=1, Maximum=2)
    Module.Param0WriteEnabled=false
    Module.Param1WriteEnabled=true
    Module.Param2WriteEnabled=false
    Module.Param3WriteEnabled=true
   ) {module=DynamicMaterialParameters002}
   /SetVariables(Particles.SpawnGroup=&Add_Integer(A=$Particles.SpawnGroup, B=2)) {isSetParameters=true}
  ]
  particleUpdateScript [
   &ParticleState()
   &SkeletalMeshLocation(
    Bone / Socket Index=$Particles.SpawnGroup
    Bone Sampling Mode=::NewEnumerator4
    Inherit Velocity=false
    Interpolate Mesh Sampling=false
    MaskSpawnGroup=false
    Skeletal Mesh=$User.Skeletal Mesh02
   ) {module=SkeletalMeshLocation001}
   &self:LERP-Source_to_Destination(
    DestinationPosition:NiagaraPosition=$Particles.Initial.Position
    DestinationPosition-Offset:Vector3f=&RandomRangeVector(
     Minimum=&VectorFromCurve(VectorCurve=DataInterface() {propertyValues {
        XCurve {keys [<at 0 -4>]}
        YCurve {keys [<at 0 -2>]}
        ZCurve {keys [<at 0 0 RCIM_Cubic>, {arriveTangent=40.1827, interpMode=RCIM_Cubic, leaveTangent=40.183, tangentMode=RCTM_User, time=1, value=20}]}
       }})
     Maximum=&VectorFromCurve(VectorCurve=DataInterface() {propertyValues {
        XCurve {keys [<at 0 4>]}
        YCurve {keys [<at 0 3>]}
        ZCurve {keys [<at 0 0 RCIM_Cubic>, {arriveTangent=36.9034, interpMode=RCIM_Cubic, leaveTangent=36.9036, tangentMode=RCTM_User, time=1, value=20}]}
       }})
    )
    Normalized Influence:NiagaraFloat=&MakeCustomFloatFromBool(
     Bool=$User.IsSkydiving
     True Float=&FloatFromCurve(
      FloatCurve=DataInterface() {propertyValues {Curve {keys [
          {arriveTangent=0.0353, interpMode=RCIM_Cubic, leaveTangent=0.0353, tangentMode=RCTM_User, time=0, value=0}
          {arriveTangent=0.0217, interpMode=RCIM_Cubic, leaveTangent=0.0217, tangentMode=RCTM_User, time=1, value=0.025}
         ]}}}
      Scale Curve=0.5
     )
     False Float=&FloatFromCurve(
      FloatCurve=DataInterface() {propertyValues {Curve {keys [
          {arriveTangent=0.1289, interpMode=RCIM_Cubic, leaveTangent=0.1289, tangentMode=RCTM_User, time=0, value=0}
          {arriveTangent=0.0487, interpMode=RCIM_Cubic, leaveTangent=0.0487, tangentMode=RCTM_User, time=1, value=0.1}
         ]}}}
      Scale Curve=0.5
     )
    )
    SourcePosition:Vector3f=$Output.SkeletalMeshLocation001.SampledPosition
    SourcePosition-Offset:Vector3f=&TransformVector(Vector=&RandomRangeVector(Minimum=<4 -1 -1>, Maximum=<5 1 2>), Destination Space=::Simulation)
   ) {module=LERP-Source_to_Destination}
   &ScaleColor(
    Scale Mode=::NewEnumerator1
    Scale RGBA=&Vector4FromCurve(Vector4Curve=DataInterface() {propertyValues {
       AlphaCurve {keys [<at 0.0024 0>,<at 0.15 1.5>,<at 0.45 1>,<at 0.75 2>,<at 1 0>]}
       BlueCurve {keys [<at 0 5>,<at 0.4751 0.9427>,<at 0.6 0.5885>,<at 0.75 0.5156>]}
       GreenCurve {keys [<at 0 5>,<at 0.4751 0.9427>,<at 0.6 0.5722>,<at 0.75 0.5018>]}
       RedCurve {keys [<at 0 5>,<at 0.4751 0.9427>,<at 0.6 0.5579>,<at 0.75 0.4968>]}
      }})
    ScaleRGBA=true
   ) {module=ScaleColor002}
   /SetVariables(Particles.SubImageIndex=&Add_Float(
     A=&RandomRangeFloat(Minimum=0, Maximum=20)
     B=&FloatFromCurve(FloatCurve=DataInterface() {propertyValues {Curve {keys [<at 0 0>,<at 1 24>]}}})
    )) {isSetParameters=true, module=SetVariables_#5}
   &ScaleSpriteSize(Scale Factor=&Vector2DFromFloat(Value=&FloatFromCurve(FloatCurve=DataInterface() {propertyValues {Curve {keys [
          {arriveTangent=5.8415, interpMode=RCIM_Cubic, leaveTangent=5.8415, tangentMode=RCTM_User, time=0, value=0}
          {arriveTangent=0.4267, interpMode=RCIM_Cubic, leaveTangent=0.4267, tangentMode=RCTM_User, time=0.25, value=1}
          {arriveTangent=-1.3333, interpMode=RCIM_Cubic, leaveTangent=-1.3333, tangentMode=RCTM_User, time=0.5752, value=1}
          {arriveTangent=-3.118, interpMode=RCIM_Cubic, leaveTangent=-3.118, tangentMode=RCTM_User, time=1, value=0.5}
         ]}}})))
   &DynamicMaterialParameters(
    Index 0 Param 1=&FloatFromCurve(FloatCurve=DataInterface() {propertyValues {Curve {keys [{arriveTangent=2.7551, interpMode=RCIM_Cubic, leaveTangent=2.7551, tangentMode=RCTM_User, time=0, value=0}, <at 0.5 1>]}}})
    Index 0 Param 2=&RandomRangeFloat(Minimum=0.33, Maximum=1)
    Index 0 Param 3=&FloatFromCurve(FloatCurve=DataInterface() {propertyValues {Curve {keys [<at 0 1>,<at 1 2>]}}})
    Index 0 Param 4=1
    Module.Param0WriteEnabled=true
    Module.Param1WriteEnabled=true
    Module.Param2WriteEnabled=true
    Module.Param3WriteEnabled=true
   ) {module=DynamicMaterialParameters001}
   &CameraOffset(Camera Offset Amount=-10) {module=CameraOffset001}
   &SpriteRotationRate(Rotation Rate=&RandomRangeFloat(Minimum=-50, Maximum=50))
  ]
  rendererClasses [&NiagaraSpriteRendererProperties]
  renderers [{rendererClass=&NiagaraSpriteRendererProperties}]
 }
 Bottom_Flame_H {
  emitterUpdateScript [
   &EmitterState()
   &SpawnRate(SpawnRate=12, Use Spawn Probability=false)
  ]
  particleSpawnScript [
   &InitializeParticle(
    Color=$User.Variant_Color
    Lifetime=&UniformRangedFloat(Minimum=0.6, Maximum=0.7)
    Module.Write Color=true
    Module.Write Lifetime=true
    Module.Write Mass=false
    Module.Write Material Random=true
    Module.Write Position=false
    Module.Write Scale=false
    Module.Write SpriteRotation=true
    Module.Write SpriteSize=true
    Position=&SimulationPosition()
    Sprite Rotation=&UniformRangedFloat(Minimum=-20, Maximum=20)
    Sprite Size=&MakeCustomVector2DFromBool(Bool=&RandomBool(), True Vector2D=<50 50>, False Vector2D=<-50 50>)
   )
   &FN_SetSpawnGroup(Number of Spawn Groups=2)
   &SkeletalMeshLocation(
    Bone / Socket Index=$Particles.SpawnGroup
    Bone Sampling Mode=::NewEnumerator4
    Inherit Velocity=false
    MaskSpawnGroup=false
    Skeletal Mesh=$User.SkeletalMesh
   )
   &SphereLocation(MaskSpawnGroup=false, Module.Override Randomness=false, Non Uniform Scale=<1 1 0.1>, Sphere Radius=5, Surface Only=false)
   &DynamicMaterialParameters(
    Index 0 Param 2=&UniformRangedFloat(Minimum=0, Maximum=1)
    Index 0 Param 3=1
    Index 0 Param 4=&UniformRangedFloat(Minimum=1, Maximum=2)
    Module.Param0WriteEnabled=false
    Module.Param1WriteEnabled=true
    Module.Param2WriteEnabled=false
    Module.Param3WriteEnabled=true
   ) {module=DynamicMaterialParameters002}
  ]
  particleUpdateScript [
   &ParticleState()
   &SkeletalMeshLocation(
    Bone / Socket Index=$Particles.SpawnGroup
    Bone Sampling Mode=::NewEnumerator4
    Inherit Velocity=false
    Interpolate Mesh Sampling=false
    MaskSpawnGroup=false
    Skeletal Mesh=$User.SkeletalMesh
   ) {module=SkeletalMeshLocation001}
   &self:LERP-Source_to_Destination(
    DestinationPosition:NiagaraPosition=$Particles.Initial.Position
    DestinationPosition-Offset:Vector3f=&RandomRangeVector(
     Minimum=&VectorFromCurve(VectorCurve=DataInterface() {propertyValues {
        XCurve {keys [<at 0 -4>]}
        YCurve {keys [<at 0 -2>]}
        ZCurve {keys [<at 0 0 RCIM_Cubic>, {arriveTangent=40.1827, interpMode=RCIM_Cubic, leaveTangent=40.183, tangentMode=RCTM_User, time=1, value=20}]}
       }})
     Maximum=&VectorFromCurve(VectorCurve=DataInterface() {propertyValues {
        XCurve {keys [<at 0 4>]}
        YCurve {keys [<at 0 3>]}
        ZCurve {keys [<at 0 0 RCIM_Cubic>, {arriveTangent=36.9034, interpMode=RCIM_Cubic, leaveTangent=36.9036, tangentMode=RCTM_User, time=1, value=20}]}
       }})
    )
    Normalized Influence:NiagaraFloat=&MakeCustomFloatFromBool(
     Bool=$User.IsSkydiving
     True Float=&FloatFromCurve(
      FloatCurve=DataInterface() {propertyValues {Curve {keys [
          {arriveTangent=0.0353, interpMode=RCIM_Cubic, leaveTangent=0.0353, tangentMode=RCTM_User, time=0, value=0}
          {arriveTangent=0.0217, interpMode=RCIM_Cubic, leaveTangent=0.0217, tangentMode=RCTM_User, time=1, value=0.025}
         ]}}}
      Scale Curve=0.5
     )
     False Float=&FloatFromCurve(
      FloatCurve=DataInterface() {propertyValues {Curve {keys [
          {arriveTangent=0.1289, interpMode=RCIM_Cubic, leaveTangent=0.1289, tangentMode=RCTM_User, time=0, value=0}
          {arriveTangent=0.0487, interpMode=RCIM_Cubic, leaveTangent=0.0487, tangentMode=RCTM_User, time=1, value=0.1}
         ]}}}
      Scale Curve=0.5
     )
    )
    SourcePosition:Vector3f=$Output.SkeletalMeshLocation001.SampledPosition
    SourcePosition-Offset:Vector3f=&TransformVector(Vector=&RandomRangeVector(Minimum=<4 -1 -1>, Maximum=<5 1 2>), Destination Space=::Simulation)
   ) {module=LERP-Source_to_Destination}
   &ScaleColor(
    Scale Mode=::NewEnumerator1
    Scale RGBA=&Vector4FromCurve(Vector4Curve=DataInterface() {propertyValues {
       AlphaCurve {keys [<at 0.0024 0>,<at 0.15 1.5>,<at 0.45 1>,<at 0.75 2>,<at 1 0>]}
       BlueCurve {keys [<at 0 5>,<at 0.4751 0.9427>,<at 0.6 0.5885>,<at 0.75 0.5156>]}
       GreenCurve {keys [<at 0 5>,<at 0.4751 0.9427>,<at 0.6 0.5722>,<at 0.75 0.5018>]}
       RedCurve {keys [<at 0 5>,<at 0.4751 0.9427>,<at 0.6 0.5579>,<at 0.75 0.4968>]}
      }})
    ScaleRGBA=true
   ) {module=ScaleColor002}
   /SetVariables(Particles.SubImageIndex=&Add_Float(
     A=&RandomRangeFloat(Minimum=0, Maximum=20)
     B=&FloatFromCurve(FloatCurve=DataInterface() {propertyValues {Curve {keys [<at 0 0>,<at 1 24>]}}})
    )) {isSetParameters=true, module=SetVariables_#5}
   &ScaleSpriteSize(Scale Factor=&Vector2DFromFloat(Value=&FloatFromCurve(FloatCurve=DataInterface() {propertyValues {Curve {keys [
          {arriveTangent=5.8415, interpMode=RCIM_Cubic, leaveTangent=5.8415, tangentMode=RCTM_User, time=0, value=0}
          {arriveTangent=0.4267, interpMode=RCIM_Cubic, leaveTangent=0.4267, tangentMode=RCTM_User, time=0.25, value=1}
          {arriveTangent=-1.3333, interpMode=RCIM_Cubic, leaveTangent=-1.3333, tangentMode=RCTM_User, time=0.5752, value=1}
          {arriveTangent=-3.118, interpMode=RCIM_Cubic, leaveTangent=-3.118, tangentMode=RCTM_User, time=1, value=0.5}
         ]}}})))
   &DynamicMaterialParameters(
    Index 0 Param 1=&FloatFromCurve(FloatCurve=DataInterface() {propertyValues {Curve {keys [{arriveTangent=2.7551, interpMode=RCIM_Cubic, leaveTangent=2.7551, tangentMode=RCTM_User, time=0, value=0}, <at 0.5 1>]}}})
    Index 0 Param 2=&RandomRangeFloat(Minimum=0.33, Maximum=1)
    Index 0 Param 3=&FloatFromCurve(FloatCurve=DataInterface() {propertyValues {Curve {keys [<at 0 1>,<at 1 2>]}}})
    Index 0 Param 4=1
    Module.Param0WriteEnabled=true
    Module.Param1WriteEnabled=true
    Module.Param2WriteEnabled=true
    Module.Param3WriteEnabled=true
   ) {module=DynamicMaterialParameters001}
   &CameraOffset(Camera Offset Amount=-10) {module=CameraOffset001}
   &SpriteRotationRate(Rotation Rate=&RandomRangeFloat(Minimum=-50, Maximum=50))
  ]
  rendererClasses [&NiagaraSpriteRendererProperties]
  renderers [{rendererClass=&NiagaraSpriteRendererProperties}]
 }
 Decal_Flame {
  emitterUpdateScript [
   &EmitterState()
   &SpawnRate(SpawnRate=4, Use Spawn Probability=false)
  ]
  particleSpawnScript [
   &InitializeParticle(
    Color=&MultiplyLinearColor(Linear Color=$User.Variant_Color, Scale Factor=<1 1 1 2>)
    Lifetime=&UniformRangedFloat(Minimum=0.5, Maximum=1)
    Module.Write Color=true
    Module.Write Lifetime=true
    Module.Write Mass=false
    Module.Write Material Random=true
    Module.Write Position=false
    Module.Write Scale=false
    Module.Write SpriteRotation=false
    Module.Write SpriteSize=true
    Position=&SimulationPosition()
    Sprite Rotation=&UniformRangedFloat(Minimum=-180, Maximum=180)
    Sprite Size=&Vector2DFromFloat(Value=&RandomRangeFloat(Minimum=25, Maximum=50))
   )
   &ApplyInitialForces()
   &DynamicMaterialParameters(
    Index 0 Param 1=1
    Index 0 Param 2=1.5
    Index 0 Param 3=&RandomRangeFloat(Minimum=0.1, Maximum=1)
    Index 0 Param 4=&RandomRangeFloat(Minimum=0.75, Maximum=1)
    Index 1 Param 1=1
    Index 1 Param 2=24
    Module.Param0WriteEnabled=true
    Module.Param0WriteEnabled1=true
    Module.Param1WriteEnabled=true
    Module.Param1WriteEnabled1=true
    Module.Param2WriteEnabled=true
    Module.Param2WriteEnabled1=true
    Module.Param3WriteEnabled=true
    Module.Param3WriteEnabled1=true
    Write Parameter Index 1=true
   )
   &CameraOffset(Camera Offset Amount=&RandomRangeFloat(Minimum=5, Maximum=15))
   &FN_SetSpawnGroup(Number of Spawn Groups=2)
   /SetVariables(Particles.SpawnGroup=&Add_Integer(A=$Particles.SpawnGroup, B=2)) {isSetParameters=true}
  ]
  particleUpdateScript [
   &ParticleState()
   &SkeletalMeshLocation(
    Bone / Socket Index=$Particles.SpawnGroup
    Bone Sampling Mode=::NewEnumerator4
    Inherit Velocity=false
    MaskSpawnGroup=false
    Sampled Position Offset Coordinate Space:ENiagaraMeshTransforms=::NewEnumerator2
    Skeletal Mesh=$User.Skeletal Mesh02
   )
   &SolveForcesAndVelocity(Clamp Velocity=false, Limit Acceleration=false)
   &ScaleColor(
    Scale Mode=::NewEnumerator1
    Scale RGBA=/Niagara/DynamicInputs/UniformRange/UniformRangedVector4(
     Minimum:Vector4f=&Vector4FromCurve(Vector4Curve=DataInterface() {propertyValues {
        AlphaCurve {keys [<at 0.0024 0>,<at 0.1271 1>,<at 0.265 0.9976>,<at 0.9977 -0.0026>]}
        BlueCurve {keys [<at 0.3937 10>,<at 0.75 0.9271>]}
        GreenCurve {keys [<at 0.3937 10>,<at 0.75 0.9271>]}
        RedCurve {keys [<at 0.3937 10>,<at 0.75 0.9271>]}
       }})
     Maximum:Vector4f=&Vector4FromCurve(Vector4Curve=DataInterface() {propertyValues {
        AlphaCurve {keys [<at 0.0024 0>,<at 0.1271 1>,<at 0.265 0.9976>,<at 0.9977 -0.0026>]}
        BlueCurve {keys [<at 0.3937 10>,<at 0.75 0.7969>]}
        GreenCurve {keys [<at 0.3937 10>,<at 0.75 0.7969>]}
        RedCurve {keys [<at 0.3937 10>,<at 0.75 0.7969>]}
       }})
     Randomness Mode:ENiagaraRandomnessMode=::NewEnumerator0
     Random Seed:NiagaraInt32=0
     Fixed Random Seed:NiagaraBool=false
    )
    ScaleRGBA=true
   ) {module=ScaleColor001}
   &ScaleSpriteSize(Scale Factor=&Vector2DFromFloat(Value=&FloatFromCurve(FloatCurve=DataInterface() {propertyValues {Curve {keys [
          {arriveTangent=1.5947, interpMode=RCIM_Cubic, leaveTangent=1.5947, tangentMode=RCTM_User, time=0, value=0.5946}
          <at 0.9886 0.9389 RCIM_Cubic>
         ]}}})))
  ]
  rendererClasses [&NiagaraSpriteRendererProperties]
  renderers [{rendererClass=&NiagaraSpriteRendererProperties}]
 }
 Decal_Flame_H {
  emitterUpdateScript [
   &EmitterState()
   &SpawnRate(SpawnRate=4, Use Spawn Probability=false)
  ]
  particleSpawnScript [
   &InitializeParticle(
    Color=&MultiplyLinearColor(Linear Color=$User.Variant_Color, Scale Factor=<1 1 1 2>)
    Lifetime=&UniformRangedFloat(Minimum=0.5, Maximum=1)
    Module.Write Color=true
    Module.Write Lifetime=true
    Module.Write Mass=false
    Module.Write Material Random=true
    Module.Write Position=false
    Module.Write Scale=false
    Module.Write SpriteRotation=false
    Module.Write SpriteSize=true
    Position=&SimulationPosition()
    Sprite Rotation=&UniformRangedFloat(Minimum=-180, Maximum=180)
    Sprite Size=&Vector2DFromFloat(Value=&RandomRangeFloat(Minimum=25, Maximum=50))
   )
   &ApplyInitialForces()
   &DynamicMaterialParameters(
    Index 0 Param 1=1
    Index 0 Param 2=1.5
    Index 0 Param 3=&RandomRangeFloat(Minimum=0.1, Maximum=1)
    Index 0 Param 4=&RandomRangeFloat(Minimum=0.75, Maximum=1)
    Index 1 Param 1=1
    Index 1 Param 2=24
    Module.Param0WriteEnabled=true
    Module.Param0WriteEnabled1=true
    Module.Param1WriteEnabled=true
    Module.Param1WriteEnabled1=true
    Module.Param2WriteEnabled=true
    Module.Param2WriteEnabled1=true
    Module.Param3WriteEnabled=true
    Module.Param3WriteEnabled1=true
    Write Parameter Index 1=true
   )
   &CameraOffset(Camera Offset Amount=&RandomRangeFloat(Minimum=5, Maximum=15))
   &FN_SetSpawnGroup(Number of Spawn Groups=2)
  ]
  particleUpdateScript [
   &ParticleState()
   &SkeletalMeshLocation(
    Bone / Socket Index=$Particles.SpawnGroup
    Bone Sampling Mode=::NewEnumerator4
    Inherit Velocity=false
    MaskSpawnGroup=false
    Sampled Position Offset=&Lerp_Vector(A=<0 -10 0>, B=<0 10 0>, Alpha=$System.Dive)
    Sampled Position Offset Coordinate Space:ENiagaraMeshTransforms=::NewEnumerator2
    Skeletal Mesh=$User.SkeletalMesh
   )
   &SolveForcesAndVelocity(Clamp Velocity=false, Limit Acceleration=false)
   &ScaleColor(
    Scale Mode=::NewEnumerator1
    Scale RGBA=&UniformRangedVector4(
     Minimum:Vector4f=&Vector4FromCurve(Vector4Curve=DataInterface() {propertyValues {
        AlphaCurve {keys [<at 0.0024 0>,<at 0.1271 1>,<at 0.265 0.9976>,<at 0.9977 -0.0026>]}
        BlueCurve {keys [<at 0.3937 10>,<at 0.75 0.9271>]}
        GreenCurve {keys [<at 0.3937 10>,<at 0.75 0.9271>]}
        RedCurve {keys [<at 0.3937 10>,<at 0.75 0.9271>]}
       }})
     Maximum:Vector4f=&Vector4FromCurve(Vector4Curve=DataInterface() {propertyValues {
        AlphaCurve {keys [<at 0.0024 0>,<at 0.1271 1>,<at 0.265 0.9976>,<at 0.9977 -0.0026>]}
        BlueCurve {keys [<at 0.3937 10>,<at 0.75 0.7969>]}
        GreenCurve {keys [<at 0.3937 10>,<at 0.75 0.7969>]}
        RedCurve {keys [<at 0.3937 10>,<at 0.75 0.7969>]}
       }})
     Randomness Mode:ENiagaraRandomnessMode=::NewEnumerator0
     Random Seed:NiagaraInt32=0
     Fixed Random Seed:NiagaraBool=false
    )
    ScaleRGBA=true
   ) {module=ScaleColor001}
   &ScaleSpriteSize(Scale Factor=&Vector2DFromFloat(Value=&FloatFromCurve(FloatCurve=DataInterface() {propertyValues {Curve {keys [
          {arriveTangent=1.5947, interpMode=RCIM_Cubic, leaveTangent=1.5947, tangentMode=RCTM_User, time=0, value=0.5946}
          <at 0.9886 0.9389 RCIM_Cubic>
         ]}}})))
  ]
  rendererClasses [&NiagaraSpriteRendererProperties]
  renderers [{rendererClass=&NiagaraSpriteRendererProperties}]
 }
 FN_Debug {
  bEnabled=false
  emitterUpdateScript [
   &EmitterState()
   &SpawnBurst_Instantaneous(Spawn Count=1, Use Spawn Probability=false)
  ]
  particleSpawnScript [&InitializeParticle(
    Lifetime=5
    Module.Write Color=true
    Module.Write Lifetime=true
    Module.Write Mass=true
    Module.Write Material Random=true
    Module.Write Position=false
    Module.Write Scale=false
    Module.Write SpriteRotation=false
    Module.Write SpriteSize=true
    Position=&SimulationPosition()
    Sprite Size=<100 100>
   )]
  particleUpdateScript [
   &ParticleState()
   &FN_Timer()
   &SkeletalMeshLocation(
    Apply Sampled Orientation to Mesh Particles=true
    Bone Sampling Mode=::NewEnumerator4
    Inherit Velocity=false
    Interpolate Mesh Sampling=false
    MaskSpawnGroup=false
    Sampled Position Offset=&Lerp_Vector(A=<0 12.5 2.5>, B=<1 0 0>, Alpha=$System.Dive)
    Sampled Position Offset Coordinate Space:ENiagaraMeshTransforms=::NewEnumerator2
    Skeletal Mesh=$User.SkeletalMesh
   ) {module=SkeletalMeshLocation001}
   /SetVariables(Particles.Debug:NiagaraFloat=$System.FN_Timer001.Value) {isSetParameters=true}
   &DynamicMaterialParameters(
    Index 0 Param 1=$Particles.Debug
    Module.Param0WriteEnabled=true
    Module.Param1WriteEnabled=true
    Module.Param2WriteEnabled=true
    Module.Param3WriteEnabled=true
   )
   &CameraOffset(Camera Offset Amount=111)
  ]
  rendererClasses [&NiagaraSpriteRendererProperties]
  renderers [{rendererClass=&NiagaraSpriteRendererProperties}]
 }
 Fire_Mesh {
  emitterUpdateScript [
   &EmitterState(
    Inactive Response=::NewEnumerator1
    Life Cycle Mode=::NewEnumerator1
    Loop Behavior=::NewEnumerator1
    Loop Duration Mode=::NewEnumerator1
    UseLoopDelay=false
   )
   &SpawnBurst_Instantaneous(Spawn Count=2, Use Spawn Probability=false)
  ]
  particleSpawnScript [
   /SetVariables(Particles.Position=&Add_Vector(A=Linked() {linkedVariable {name=Engine.Owner.Position, type=NiagaraPosition}}, B=<0 0 0>)) {isSetParameters=true, module=SetVariables_#6}
   /SetVariables(Particles.PendulumDrag=5, Particles.PendulumLengthScale=1, Particles.PendulumMaskFalloff=3, Particles.PendulumRigidity=0.05) {isSetParameters=true, module=SetVariables_#7}
   &InitializeParticle(
    Lifetime=2
    Module.Write Color=true
    Module.Write Lifetime=true
    Module.Write Mass=true
    Module.Write Material Random=true
    Module.Write Position=false
    Module.Write Scale=true
    Module.Write SpriteRotation=false
    Module.Write SpriteSize=true
   )
   &FN_SetSpawnGroup(Number of Spawn Groups=2)
   /SetVariables(Particles.SpawnGroup=&Add_Integer(A=$Particles.SpawnGroup, B=2)) {isSetParameters=true}
  ]
  particleUpdateScript [
   &SkeletalMeshLocation(
    Apply Sampled Orientation to Mesh Particles=true
    Apply To Particle Position=false
    Bone / Socket Index=$Particles.SpawnGroup
    Bone Sampling Mode=::NewEnumerator4
    Inherit Velocity=false
    MaskSpawnGroup=false
    Skeletal Mesh=$User.Skeletal Mesh02
   )
   &ParticleState(Kill Particles When Lifetime Has Elapsed=false)
   /SetVariables(
    Particles.MeshOrigin:Vector3f=&Add_Vector(A=$Output.SkeletalMeshLocation.SampledPosition, B=<0 0 0>)
    Particles.NeutralRestingAxis:Vector3f=<0 0 1>
   ) {isSetParameters=true, module=SetVariables_#8}
   &GravityForce(Coordinate Space=::Local, Gravity=<2500 -250 0>)
   &Drag(Drag=$Particles.PendulumDrag, Use Linear Drag=true, Use Rotational Drag=false)
   &PendulumSetup(
    Module.Enable Spring Driver=true
    Pendulum Length:NiagaraFloat=&Multiply_Float(A=60, B=$Particles.PendulumLengthScale)
    Pendulum Pivot:Vector3f=$Particles.MeshOrigin
    Pendulum Rest Axis:Vector3f=$Particles.NeutralRestingAxis
    "Spring Driven Constraint (Tightness)"=2
   )
   &SolveForcesAndVelocity(Clamp Velocity=false, Limit Acceleration=false)
   &PendulumConstraint()
   &CalculateAccurateVelocity()
   /SetVariables(Particles.MeshFacing:Vector3f=&NormalizeVector(Vector To Normalize=&Subtract_Vector(A=$Particles.MeshOrigin, B=Linked() {linkedVariable {name=Particles.Position, type=NiagaraPosition}}))) {isSetParameters=true, module=SetVariables_#9}
   /SetVariables(Particles.MeshOrientation=&MultiplyQuaternion(
     Quaternion A=$Output.SkeletalMeshLocation.SampledRotation
     Quaternion B=&MakeQuatFromAxisAngle(
      Axis=&SelectVectorFromArray(
       Direct Array Index=&Add_Integer(A=$Particles.SpawnGroup, B=-2)
       Vector Selection Array=DataInterface() {propertyValues {FloatData [<0 0 1>, <0 0 0>]}}
      )
      AngleInDegrees=180
     )
    )) {isSetParameters=true, module=SetVariables_#10}
   &DynamicMaterialParameters(
    Index 0 Param 1=&MakeFloatFromVector(Vector=Linked() {linkedVariable {name=Particles.Previous.Position, type=NiagaraPosition}}, Channel=::NewEnumerator0)
    Index 0 Param 2=&MakeFloatFromVector(Vector=Linked() {linkedVariable {name=Particles.Previous.Position, type=NiagaraPosition}}, Channel=::NewEnumerator1)
    Index 0 Param 3=&MakeFloatFromVector(Vector=Linked() {linkedVariable {name=Particles.Previous.Position, type=NiagaraPosition}}, Channel=::NewEnumerator1)
    Index 0 Param 4=1
    Index 1 Param 1=&MakeFloatFromVector(Vector=$Particles.NeutralRestingAxis, Channel=::NewEnumerator0)
    Index 1 Param 2=&MakeFloatFromVector(Vector=$Particles.NeutralRestingAxis, Channel=::NewEnumerator1)
    Index 1 Param 3=&MakeFloatFromVector(Vector=$Particles.NeutralRestingAxis, Channel=::NewEnumerator2)
    Index 1 Param 4=1
    Index 2 Param 1=HlslExpression() {hlslExpression=Particles.MeshOrigin.x}
    Index 2 Param 2=HlslExpression() {hlslExpression=Particles.MeshOrigin.y}
    Index 2 Param 3=HlslExpression() {hlslExpression=Particles.MeshOrigin.z}
    Index 2 Param 4=$Particles.PendulumMaskFalloff
    Index 3 Param 1=&MakeFloatFromVector(Vector=Linked() {linkedVariable {name=Particles.Position, type=NiagaraPosition}}, Channel=::NewEnumerator0)
    Index 3 Param 2=&MakeFloatFromVector(Vector=Linked() {linkedVariable {name=Particles.Position, type=NiagaraPosition}}, Channel=::NewEnumerator1)
    Index 3 Param 3=&MakeFloatFromVector(Vector=Linked() {linkedVariable {name=Particles.Position, type=NiagaraPosition}}, Channel=::NewEnumerator2)
    Index 3 Param 4=$Particles.PendulumRigidity
    Module.Param0WriteEnabled=true
    Module.Param0WriteEnabled1=true
    Module.Param0WriteEnabled2=true
    Module.Param0WriteEnabled3=true
    Module.Param1WriteEnabled=true
    Module.Param1WriteEnabled1=true
    Module.Param1WriteEnabled2=true
    Module.Param1WriteEnabled3=true
    Module.Param2WriteEnabled=true
    Module.Param2WriteEnabled1=true
    Module.Param2WriteEnabled2=true
    Module.Param2WriteEnabled3=true
    Module.Param3WriteEnabled=true
    Module.Param3WriteEnabled1=true
    Module.Param3WriteEnabled2=true
    Module.Param3WriteEnabled3=true
    Write Parameter Index 1=true
    Write Parameter Index 2=true
    Write Parameter Index 3=true
   )
   &CameraOffset(Camera Offset Amount=10)
  ]
  rendererClasses [&NiagaraSpriteRendererProperties, &NiagaraMeshRendererProperties]
  renderers [
   {rendererClass=&NiagaraSpriteRendererProperties}
   {rendererClass=&NiagaraMeshRendererProperties, rendererIndex=1}
  ]
 }
 Fire_Mesh_H {
  emitterUpdateScript [
   &EmitterState(
    Inactive Response=::NewEnumerator1
    Life Cycle Mode=::NewEnumerator1
    Loop Behavior=::NewEnumerator1
    Loop Duration Mode=::NewEnumerator1
    UseLoopDelay=false
   )
   &SpawnBurst_Instantaneous(Spawn Count=2, Use Spawn Probability=false)
  ]
  particleSpawnScript [
   /SetVariables(Particles.Position=&Add_Vector(A=Linked() {linkedVariable {name=Engine.Owner.Position, type=NiagaraPosition}}, B=<0 0 0>)) {isSetParameters=true, module=SetVariables_#6}
   /SetVariables(Particles.PendulumDrag=5, Particles.PendulumLengthScale=1, Particles.PendulumMaskFalloff=3, Particles.PendulumRigidity=0.05) {isSetParameters=true, module=SetVariables_#7}
   &InitializeParticle(
    Lifetime=2
    Mesh Scale=<1.2 1.2 1.2>
    Module.Write Color=true
    Module.Write Lifetime=true
    Module.Write Mass=true
    Module.Write Material Random=true
    Module.Write Position=false
    Module.Write Scale=true
    Module.Write SpriteRotation=false
    Module.Write SpriteSize=true
   )
   &FN_SetSpawnGroup(Number of Spawn Groups=2)
  ]
  particleUpdateScript [
   &SkeletalMeshLocation(
    Apply Sampled Orientation to Mesh Particles=true
    Apply To Particle Position=false
    Bone / Socket Index=$Particles.SpawnGroup
    Bone Sampling Mode=::NewEnumerator4
    Inherit Velocity=false
    MaskSpawnGroup=false
    Skeletal Mesh=$User.SkeletalMesh
   )
   &ParticleState(Kill Particles When Lifetime Has Elapsed=false)
   /SetVariables(
    Particles.MeshOrigin:Vector3f=&Add_Vector(A=$Output.SkeletalMeshLocation.SampledPosition, B=<0 0 0>)
    Particles.NeutralRestingAxis:Vector3f=<0 0 1>
   ) {isSetParameters=true, module=SetVariables_#8}
   &GravityForce(Coordinate Space=::Local, Gravity=<2500 -250 0>)
   &Drag(Drag=$Particles.PendulumDrag, Use Linear Drag=true, Use Rotational Drag=false)
   &PendulumSetup(
    Module.Enable Spring Driver=true
    Pendulum Length:NiagaraFloat=&Multiply_Float(A=60, B=$Particles.PendulumLengthScale)
    Pendulum Pivot:Vector3f=$Particles.MeshOrigin
    Pendulum Rest Axis:Vector3f=$Particles.NeutralRestingAxis
    "Spring Driven Constraint (Tightness)"=2
   )
   &SolveForcesAndVelocity(Clamp Velocity=false, Limit Acceleration=false)
   &PendulumConstraint()
   &CalculateAccurateVelocity()
   /SetVariables(Particles.MeshFacing:Vector3f=&NormalizeVector(Vector To Normalize=&Subtract_Vector(A=$Particles.MeshOrigin, B=Linked() {linkedVariable {name=Particles.Position, type=NiagaraPosition}}))) {isSetParameters=true, module=SetVariables_#9}
   /SetVariables(Particles.MeshOrientation=&MultiplyQuaternion(
     Quaternion A=$Output.SkeletalMeshLocation.SampledRotation
     Quaternion B=&MakeQuatFromAxisAngle(
      Axis=&SelectVectorFromArray(Direct Array Index=$Particles.SpawnGroup, Vector Selection Array=DataInterface() {propertyValues {FloatData [<0 0 0>, <0 0 1>]}})
      AngleInDegrees=180
     )
    )) {isSetParameters=true, module=SetVariables_#10}
   &DynamicMaterialParameters(
    Index 0 Param 1=&MakeFloatFromVector(Vector=Linked() {linkedVariable {name=Particles.Previous.Position, type=NiagaraPosition}}, Channel=::NewEnumerator0)
    Index 0 Param 2=&MakeFloatFromVector(Vector=Linked() {linkedVariable {name=Particles.Previous.Position, type=NiagaraPosition}}, Channel=::NewEnumerator1)
    Index 0 Param 3=&MakeFloatFromVector(Vector=Linked() {linkedVariable {name=Particles.Previous.Position, type=NiagaraPosition}}, Channel=::NewEnumerator1)
    Index 0 Param 4=1
    Index 1 Param 1=&MakeFloatFromVector(Vector=$Particles.NeutralRestingAxis, Channel=::NewEnumerator0)
    Index 1 Param 2=&MakeFloatFromVector(Vector=$Particles.NeutralRestingAxis, Channel=::NewEnumerator1)
    Index 1 Param 3=&MakeFloatFromVector(Vector=$Particles.NeutralRestingAxis, Channel=::NewEnumerator2)
    Index 1 Param 4=1
    Index 2 Param 1=HlslExpression() {hlslExpression=Particles.MeshOrigin.x}
    Index 2 Param 2=HlslExpression() {hlslExpression=Particles.MeshOrigin.y}
    Index 2 Param 3=HlslExpression() {hlslExpression=Particles.MeshOrigin.z}
    Index 2 Param 4=$Particles.PendulumMaskFalloff
    Index 3 Param 1=&MakeFloatFromVector(Vector=Linked() {linkedVariable {name=Particles.Position, type=NiagaraPosition}}, Channel=::NewEnumerator0)
    Index 3 Param 2=&MakeFloatFromVector(Vector=Linked() {linkedVariable {name=Particles.Position, type=NiagaraPosition}}, Channel=::NewEnumerator1)
    Index 3 Param 3=&MakeFloatFromVector(Vector=Linked() {linkedVariable {name=Particles.Position, type=NiagaraPosition}}, Channel=::NewEnumerator2)
    Index 3 Param 4=$Particles.PendulumRigidity
    Module.Param0WriteEnabled=true
    Module.Param0WriteEnabled1=true
    Module.Param0WriteEnabled2=true
    Module.Param0WriteEnabled3=true
    Module.Param1WriteEnabled=true
    Module.Param1WriteEnabled1=true
    Module.Param1WriteEnabled2=true
    Module.Param1WriteEnabled3=true
    Module.Param2WriteEnabled=true
    Module.Param2WriteEnabled1=true
    Module.Param2WriteEnabled2=true
    Module.Param2WriteEnabled3=true
    Module.Param3WriteEnabled=true
    Module.Param3WriteEnabled1=true
    Module.Param3WriteEnabled2=true
    Module.Param3WriteEnabled3=true
    Write Parameter Index 1=true
    Write Parameter Index 2=true
    Write Parameter Index 3=true
   )
   &CameraOffset(Camera Offset Amount=10)
  ]
  rendererClasses [&NiagaraSpriteRendererProperties, &NiagaraMeshRendererProperties]
  renderers [
   {rendererClass=&NiagaraSpriteRendererProperties}
   {rendererClass=&NiagaraMeshRendererProperties, rendererIndex=1}
  ]
 }
 FloatingBits {
  emitterUpdateScript [
   &EmitterState()
   &SpawnBurst_Instantaneous(
    Spawn Count=2
    Spawn Probability=HlslExpression() {hlslExpression="Emitter.LoopCount == 0"}
    Use Spawn Probability=true
    UseLoopCountLimit=false
   )
  ]
  particleSpawnScript [
   /Niagara/Modules/Spawn/Initialization/V2/InitializeParticle(
    Lifetime=0.5
    Mass=&Lerp_Float(
     A=0.1
     B=2
     Alpha=&FloatFromCurve(
      FloatCurve=DataInterface() {propertyValues {Curve {keys [
          {interpMode=RCIM_Cubic, tangentMode=RCTM_User, time=0, value=0}
          {arriveTangent=3.2438, interpMode=RCIM_Cubic, leaveTangent=3.2438, tangentMode=RCTM_User, time=1, value=1}
         ]}}}
      CurveIndex=&RandomRangeFloat(Minimum=0, Maximum=1)
     )
    )
    Mass Mode=::NewEnumerator1
    Mesh Scale Mode=::NewEnumerator3
    Mesh Uniform Scale=5.5
    UsePositionOffset=false
    Write Material Random=true
   )
   &FN_SetSpawnGroup(Number of Spawn Groups=2)
   &SkeletalMeshLocation(
    Bone / Socket Index=$Particles.SpawnGroup
    Bone Sampling Mode=::NewEnumerator4
    Inherit Velocity=false
    Interpolate Mesh Sampling=false
    MaskSpawnGroup=false
    Sampled Position Offset Coordinate Space:ENiagaraMeshTransforms=::NewEnumerator2
    Skeletal Mesh=$User.SkeletalMesh
   )
   &CameraOffset(Camera Offset Amount=&MakeCustomFloatFromBool(Bool=$User.bIsFrontEndPreview, True Float=2, False Float=&RandomRangeFloat(Minimum=2.5, Maximum=7.5)))
   &InitialMeshOrientation(
    Mesh Axis To Orient=&SelectVectorFromArray(Direct Array Index=$Particles.SpawnGroup, Vector Selection Array=DataInterface() {propertyValues {FloatData [<-1 0 0>, <0 1 0>]}})
    Mesh Orientation Mode=::NewEnumerator2
    Orientation Vector=&SelectVectorFromArray(Direct Array Index=$Particles.SpawnGroup, Vector Selection Array=DataInterface() {propertyValues {FloatData [<0 0 0>, <0 1 0>]}})
    Use Rotation Vector=false
   )
  ]
  particleUpdateScript [
   &ParticleState(
    Kill Particles When Lifetime Has Elapsed=false
    Let Infinitely Lived Particles Die When Emitter Deactivates=true
    Loop Particles Lifetime=true
   )
   &DynamicMaterialParameters(
    Index 0 Param 1=1.5
    Index 0 Param 2=$System.FN_Timer001.Value
    Module.Param0WriteEnabled=true
    Module.Param1WriteEnabled=true
    Module.Param2WriteEnabled=true
    Module.Param3WriteEnabled=true
   )
   &ScaleMeshSize(Scale Factor=&SelectVectorFromArray(Direct Array Index=$Particles.SpawnGroup, Vector Selection Array=DataInterface() {propertyValues {FloatData [<1 1 1>, <1 1 1>]}})) {module=ScaleMeshSize001}
   &ScaleMeshSize(Scale Factor=&Multiply_VectorByFloat(
     Vector=$Engine.Owner.Scale
     Float=&FloatFromCurve(
      FloatCurve=DataInterface() {propertyValues {Curve {keys [
          {interpMode=RCIM_Cubic, tangentMode=RCTM_User, time=0, value=0.9}
          {interpMode=RCIM_Cubic, tangentMode=RCTM_User, time=0.5, value=1}
          {tangentMode=RCTM_User, time=1, value=0.9}
         ]}}}
      CurveIndex=$Emitter.NormalizedLoopAge
     )
    ))
   &SkeletalMeshLocation(
    Bone / Socket Index=$Particles.SpawnGroup
    Bone Sampling Mode=::NewEnumerator4
    Inherit Velocity=false
    MaskSpawnGroup=false
    Sampled Position Offset=&Lerp_Vector(
     A=<0 -2.5 15>
     B=&FN_FrontEndVector3Modifier(Gameplay=<0 -2.5 7.5>, Front End=<0 -15 7.5>)
     Alpha=$System.Dive
    )
    Sampled Position Offset Coordinate Space:ENiagaraMeshTransforms=::NewEnumerator2
    Skeletal Mesh=$User.SkeletalMesh
   ) {module=SkeletalMeshLocation001}
   /SetVariables(Particles.MeshOrientation=&MultiplyQuaternion(
     Quaternion A=$Output.SkeletalMeshLocation001.SampledRotation
     Quaternion B=&MultiplyQuaternion(
      Quaternion A=$Particles.Initial.MeshOrientation
      Quaternion B=/Niagara/DynamicInputs/TypeConversions/MakeCustomQuaternionFromBool(
       Bool:NiagaraBool=$User.bIsFrontEndPreview
       True Quat:Quat4f=&MakeQuatFromAxisAngle(
        Axis=&Lerp_Vector(
         A=&SelectVectorFromArray(Direct Array Index=$Particles.SpawnGroup, Vector Selection Array=DataInterface() {propertyValues {FloatData [<1 0.5 0.5>, <1 0.5 -0.5>]}})
         B=<1 0 0>
         Alpha=$System.Dive
        )
        AngleInDegrees=&Lerp_Float(
         A=&SelectFloatFromArray(Direct Array Index=$Particles.SpawnGroup, Float Selection Array=DataInterface() {propertyValues {FloatData [90,-90]}})
         B=&SelectFloatFromArray(Direct Array Index=$Particles.SpawnGroup, Float Selection Array=DataInterface() {propertyValues {FloatData [90,-90]}})
         Alpha=$System.Dive
        )
        NormalizeAxis=$User.bIsFrontEndPreview
       )
       False Quat:Quat4f=&MakeQuatFromAxisAngle(
        Axis=<1 0 0>
        AngleInDegrees=&SelectFloatFromArray(Direct Array Index=$Particles.SpawnGroup, Float Selection Array=DataInterface() {propertyValues {FloatData [90,-90]}})
       )
      )
     )
    )) {isSetParameters=true}
   &Drag(Drag=20, Use Linear Drag=true, Use Rotational Drag=true)
   &SolveForcesAndVelocity(Clamp Velocity=false, Limit Acceleration=false)
  ]
  rendererClasses [&NiagaraMeshRendererProperties]
  renderers [{rendererClass=&NiagaraMeshRendererProperties, rendererIndex=0}]
 }
 Front_Flame {
  emitterUpdateScript [
   &EmitterState()
   &SpawnRate(SpawnRate=16, Use Spawn Probability=false)
  ]
  particleSpawnScript [
   /Niagara/Modules/Spawn/Initialization/InitializeParticle(
    Color=$User.Variant_Color
    Lifetime=&UniformRangedFloat(Minimum=0.55, Maximum=0.65)
    Module.Write Color=true
    Module.Write Lifetime=true
    Module.Write Mass=false
    Module.Write Material Random=true
    Module.Write Position=false
    Module.Write Scale=false
    Module.Write SpriteRotation=true
    Module.Write SpriteSize=true
    Position=&SimulationPosition()
    Sprite Rotation=&UniformRangedFloat(Minimum=0, Maximum=360)
    Sprite Size=&MakeCustomVector2DFromBool(Bool=&RandomBool(), True Vector2D=<28 28>, False Vector2D=<-28 28>)
   )
   &FN_SetSpawnGroup(Number of Spawn Groups=2)
   &SkeletalMeshLocation(
    Apply To Particle Position=false
    Bone / Socket Index=$Particles.SpawnGroup
    Bone Sampling Mode=::NewEnumerator4
    Inherit Velocity=false
    MaskSpawnGroup=false
    Skeletal Mesh=$User.SkeletalMesh
   )
   &SphereLocation(MaskSpawnGroup=false, Module.Override Randomness=false, Non Uniform Scale=<1 1 0.1>, Sphere Radius=20, Surface Only=false)
   &DynamicMaterialParameters(
    Index 0 Param 2=&UniformRangedFloat(Minimum=0, Maximum=1)
    Index 0 Param 3=1
    Index 0 Param 4=&UniformRangedFloat(Minimum=1, Maximum=2)
    Module.Param0WriteEnabled=false
    Module.Param1WriteEnabled=true
    Module.Param2WriteEnabled=false
    Module.Param3WriteEnabled=true
   ) {module=DynamicMaterialParameters002}
   /SetVariables(Particles.SpawnGroup=&Add_Integer(A=$Particles.SpawnGroup, B=2)) {isSetParameters=true}
  ]
  particleUpdateScript [
   &ParticleState()
   &SkeletalMeshLocation(
    Apply To Particle Position=false
    Bone / Socket Index=$Particles.SpawnGroup
    Bone Sampling Mode=::NewEnumerator4
    Inherit Velocity=false
    MaskSpawnGroup=false
    Sampled Position Offset:NiagaraPosition=$Particles.Initial.Position
    Skeletal Mesh=$User.Skeletal Mesh02
   ) {module=SkeletalMeshLocation001}
   &self:LERP-Source_to_Destination_01(
    DestinationPosition:NiagaraPosition=$Particles.Initial.Position
    DestinationPosition-Offset:Vector3f=&RandomRangeVector(
     Minimum=&VectorFromCurve(VectorCurve=DataInterface() {propertyValues {
        XCurve {keys [<at 0 -3>]}
        YCurve {keys [<at 0 -3>]}
        ZCurve {keys [<at 0 0 RCIM_Cubic>, {arriveTangent=42.1643, interpMode=RCIM_Cubic, leaveTangent=42.1645, tangentMode=RCTM_User, time=1, value=15}]}
       }})
     Maximum=&VectorFromCurve(VectorCurve=DataInterface() {propertyValues {
        XCurve {keys [<at 0 3>]}
        YCurve {keys [<at 0 3>]}
        ZCurve {keys [<at 0 0 RCIM_Cubic>, {arriveTangent=47.8845, interpMode=RCIM_Cubic, leaveTangent=47.885, tangentMode=RCTM_User, time=1, value=15}]}
       }})
    )
    Normalized Influence:NiagaraFloat=&MakeCustomFloatFromBool(
     Bool=$User.IsSkydiving
     True Float=&FloatFromCurve(
      FloatCurve=DataInterface() {propertyValues {Curve {keys [
          {arriveTangent=0.0075, interpMode=RCIM_Cubic, leaveTangent=0.0075, tangentMode=RCTM_User, time=0, value=0}
          {arriveTangent=0.0813, interpMode=RCIM_Cubic, leaveTangent=0.0813, tangentMode=RCTM_User, time=1, value=0.025}
         ]}}}
      Scale Curve=0.5
     )
     False Float=&FloatFromCurve(
      FloatCurve=DataInterface() {propertyValues {Curve {keys [
          {arriveTangent=0.037, interpMode=RCIM_Cubic, leaveTangent=0.037, tangentMode=RCTM_User, time=0, value=0}
          {arriveTangent=0.2447, interpMode=RCIM_Cubic, leaveTangent=0.2447, tangentMode=RCTM_User, time=1, value=0.1}
         ]}}}
      Scale Curve=0.5
     )
    )
    SourcePosition:Vector3f=$Output.SkeletalMeshLocation001.SampledPosition
    SourcePosition-Offset:Vector3f=&TransformVector(Vector=&RandomRangeVector(Minimum=<0.5 -0.25 0>, Maximum=<0.5 0.25 0>), Destination Space=::Simulation)
   ) {module=LERP-Source_to_Destination_01}
   &ScaleColor(
    Scale Mode=::NewEnumerator1
    Scale RGBA=&Vector4FromCurve(Vector4Curve=DataInterface() {propertyValues {
       AlphaCurve {keys [<at 0.0024 0>,<at 0.1271 1.2>,<at 0.75 10>,<at 0.9658 0>]}
       BlueCurve {keys [<at 0.2365 1>,<at 0.65 0.8073>,<at 0.85 0.474>]}
       GreenCurve {keys [<at 0.2365 1>,<at 0.65 0.8073>,<at 0.85 0.4591>]}
       RedCurve {keys [<at 0.2365 1>,<at 0.65 0.8073>,<at 0.85 0.462>]}
      }})
    ScaleRGBA=true
   )
   /SetVariables(Particles.SubImageIndex=&FloatFromCurve(FloatCurve=DataInterface() {propertyValues {Curve {keys [<at 0 0>,<at 1 34>]}}})) {isSetParameters=true, module=SetVariables_#5}
   &ScaleSpriteSize(Scale Factor=&Vector2DFromFloat(Value=&FloatFromCurve(FloatCurve=DataInterface() {propertyValues {Curve {keys [
          {arriveTangent=7.5764, interpMode=RCIM_Cubic, leaveTangent=7.5765, tangentMode=RCTM_User, time=0, value=0.02}
          {arriveTangent=-0.0353, interpMode=RCIM_Cubic, leaveTangent=-0.0353, tangentMode=RCTM_User, time=0.2, value=1.2}
          <at 0.5752 1.1643>
          <at 1 0.65 RCIM_Cubic>
         ]}}})))
   &DynamicMaterialParameters(
    Index 0 Param 1=&FloatFromCurve(FloatCurve=DataInterface() {propertyValues {Curve {keys [{arriveTangent=2.7551, interpMode=RCIM_Cubic, leaveTangent=2.7551, tangentMode=RCTM_User, time=0, value=0}, <at 0.5 1>]}}})
    Index 0 Param 2=&RandomRangeFloat(Minimum=0.33, Maximum=1)
    Index 0 Param 3=&FloatFromCurve(FloatCurve=DataInterface() {propertyValues {Curve {keys [<at 0.8 1>,<at 1 2>]}}})
    Index 0 Param 4=1
    Module.Param0WriteEnabled=true
    Module.Param1WriteEnabled=true
    Module.Param2WriteEnabled=true
    Module.Param3WriteEnabled=true
   ) {module=DynamicMaterialParameters001}
   &CameraOffset(Camera Offset Amount=&FloatFromCurve(FloatCurve=DataInterface() {propertyValues {Curve {keys [<at 0.35 0>,<at 0.7 6>]}}})) {module=CameraOffset001}
  ]
  rendererClasses [&NiagaraSpriteRendererProperties]
  renderers [{rendererClass=&NiagaraSpriteRendererProperties}]
 }
 Front_Flame_H {
  emitterUpdateScript [
   &EmitterState()
   &SpawnRate(SpawnRate=16, Use Spawn Probability=false)
  ]
  particleSpawnScript [
   &InitializeParticle(
    Color=$User.Variant_Color
    Lifetime=&UniformRangedFloat(Minimum=0.55, Maximum=0.65)
    Module.Write Color=true
    Module.Write Lifetime=true
    Module.Write Mass=false
    Module.Write Material Random=true
    Module.Write Position=false
    Module.Write Scale=false
    Module.Write SpriteRotation=true
    Module.Write SpriteSize=true
    Position=&SimulationPosition()
    Sprite Rotation=&UniformRangedFloat(Minimum=0, Maximum=360)
    Sprite Size=&MakeCustomVector2DFromBool(Bool=&RandomBool(), True Vector2D=<28 28>, False Vector2D=<-28 28>)
   )
   &FN_SetSpawnGroup(Number of Spawn Groups=2)
   &SkeletalMeshLocation(
    Apply To Particle Position=false
    Bone / Socket Index=$Particles.SpawnGroup
    Bone Sampling Mode=::NewEnumerator4
    Inherit Velocity=false
    MaskSpawnGroup=false
    Skeletal Mesh=$User.SkeletalMesh
   )
   &SphereLocation(MaskSpawnGroup=false, Module.Override Randomness=false, Non Uniform Scale=<1 1 0.1>, Sphere Radius=20, Surface Only=false)
   &DynamicMaterialParameters(
    Index 0 Param 2=&UniformRangedFloat(Minimum=0, Maximum=1)
    Index 0 Param 3=1
    Index 0 Param 4=&UniformRangedFloat(Minimum=1, Maximum=2)
    Module.Param0WriteEnabled=false
    Module.Param1WriteEnabled=true
    Module.Param2WriteEnabled=false
    Module.Param3WriteEnabled=true
   ) {module=DynamicMaterialParameters002}
  ]
  particleUpdateScript [
   &ParticleState()
   &SkeletalMeshLocation(
    Apply To Particle Position=false
    Bone / Socket Index=$Particles.SpawnGroup
    Bone Sampling Mode=::NewEnumerator4
    Inherit Velocity=false
    MaskSpawnGroup=false
    Sampled Position Offset:NiagaraPosition=$Particles.Initial.Position
    Skeletal Mesh=$User.SkeletalMesh
   ) {module=SkeletalMeshLocation001}
   &self:LERP-Source_to_Destination_01(
    DestinationPosition:NiagaraPosition=$Particles.Initial.Position
    DestinationPosition-Offset:Vector3f=&RandomRangeVector(
     Minimum=&VectorFromCurve(VectorCurve=DataInterface() {propertyValues {
        XCurve {keys [<at 0 -3>]}
        YCurve {keys [<at 0 -3>]}
        ZCurve {keys [<at 0 0 RCIM_Cubic>, {arriveTangent=42.1643, interpMode=RCIM_Cubic, leaveTangent=42.1645, tangentMode=RCTM_User, time=1, value=15}]}
       }})
     Maximum=&VectorFromCurve(VectorCurve=DataInterface() {propertyValues {
        XCurve {keys [<at 0 3>]}
        YCurve {keys [<at 0 3>]}
        ZCurve {keys [<at 0 0 RCIM_Cubic>, {arriveTangent=47.8845, interpMode=RCIM_Cubic, leaveTangent=47.885, tangentMode=RCTM_User, time=1, value=15}]}
       }})
    )
    Normalized Influence:NiagaraFloat=&MakeCustomFloatFromBool(
     Bool=$User.IsSkydiving
     True Float=&FloatFromCurve(
      FloatCurve=DataInterface() {propertyValues {Curve {keys [
          {arriveTangent=0.0075, interpMode=RCIM_Cubic, leaveTangent=0.0075, tangentMode=RCTM_User, time=0, value=0}
          {arriveTangent=0.0813, interpMode=RCIM_Cubic, leaveTangent=0.0813, tangentMode=RCTM_User, time=1, value=0.025}
         ]}}}
      Scale Curve=0.5
     )
     False Float=&FloatFromCurve(
      FloatCurve=DataInterface() {propertyValues {Curve {keys [
          {arriveTangent=0.037, interpMode=RCIM_Cubic, leaveTangent=0.037, tangentMode=RCTM_User, time=0, value=0}
          {arriveTangent=0.2447, interpMode=RCIM_Cubic, leaveTangent=0.2447, tangentMode=RCTM_User, time=1, value=0.1}
         ]}}}
      Scale Curve=0.5
     )
    )
    SourcePosition:Vector3f=$Output.SkeletalMeshLocation001.SampledPosition
    SourcePosition-Offset:Vector3f=&TransformVector(Vector=&RandomRangeVector(Minimum=<0.5 -0.25 0>, Maximum=<0.5 0.25 0>), Destination Space=::Simulation)
   ) {module=LERP-Source_to_Destination_01}
   &ScaleColor(
    Scale Mode=::NewEnumerator1
    Scale RGBA=&Vector4FromCurve(Vector4Curve=DataInterface() {propertyValues {
       AlphaCurve {keys [<at 0.0024 0>,<at 0.1271 1.2>,<at 0.75 10>,<at 0.9658 0>]}
       BlueCurve {keys [<at 0.2365 1>,<at 0.65 0.8073>,<at 0.85 0.474>]}
       GreenCurve {keys [<at 0.2365 1>,<at 0.65 0.8073>,<at 0.85 0.4271>]}
       RedCurve {keys [<at 0.2365 1>,<at 0.65 0.8073>,<at 0.85 0.4361>]}
      }})
    ScaleRGBA=true
   )
   /SetVariables(Particles.SubImageIndex=&FloatFromCurve(FloatCurve=DataInterface() {propertyValues {Curve {keys [<at 0 0>,<at 1 34>]}}})) {isSetParameters=true, module=SetVariables_#5}
   &ScaleSpriteSize(Scale Factor=&Vector2DFromFloat(Value=&FloatFromCurve(FloatCurve=DataInterface() {propertyValues {Curve {keys [
          {arriveTangent=7.5764, interpMode=RCIM_Cubic, leaveTangent=7.5765, tangentMode=RCTM_User, time=0, value=0.02}
          {arriveTangent=-0.0353, interpMode=RCIM_Cubic, leaveTangent=-0.0353, tangentMode=RCTM_User, time=0.2, value=1.2}
          <at 0.5752 1.1643>
          <at 1 0.65 RCIM_Cubic>
         ]}}})))
   &DynamicMaterialParameters(
    Index 0 Param 1=&FloatFromCurve(FloatCurve=DataInterface() {propertyValues {Curve {keys [{arriveTangent=2.7551, interpMode=RCIM_Cubic, leaveTangent=2.7551, tangentMode=RCTM_User, time=0, value=0}, <at 0.5 1>]}}})
    Index 0 Param 2=&RandomRangeFloat(Minimum=0.33, Maximum=1)
    Index 0 Param 3=&FloatFromCurve(FloatCurve=DataInterface() {propertyValues {Curve {keys [<at 0.8 1>,<at 1 2>]}}})
    Index 0 Param 4=1
    Module.Param0WriteEnabled=true
    Module.Param1WriteEnabled=true
    Module.Param2WriteEnabled=true
    Module.Param3WriteEnabled=true
   ) {module=DynamicMaterialParameters001}
   &CameraOffset(Camera Offset Amount=&FloatFromCurve(FloatCurve=DataInterface() {propertyValues {Curve {keys [<at 0.35 0>,<at 0.7 6>]}}})) {module=CameraOffset001}
   &ScaleColor(ScaleA=true, ScaleRGB=true) {enabled=false, module=ScaleColor001}
  ]
  rendererClasses [&NiagaraSpriteRendererProperties]
  renderers [{rendererClass=&NiagaraSpriteRendererProperties}]
 }
 Light {
  emitterUpdateScript [
   &EmitterState(
    Inactive Response=::NewEnumerator1
    Life Cycle Mode=::NewEnumerator1
    Loop Behavior=::NewEnumerator1
    Loop Duration Mode=::NewEnumerator1
    UseLoopDelay=false
   )
   &SpawnBurst_Instantaneous(Spawn Count=2, Use Spawn Probability=false)
  ]
  particleSpawnScript [
   &FN_SetSpawnGroup(Number of Spawn Groups=2)
   /Niagara/Modules/Spawn/Initialization/V2/InitializeParticle(
    Color=&MultiplyLinearColor(Linear Color=$User.Variant_Color, Scale Factor=<5 5 5 1>)
    UsePositionOffset=false
    Write Material Random=true
   )
  ]
  particleUpdateScript [
   /Game/Effects/Niagara/Modules/Skydiving/FN_DivingAttributes/FN_DivingAttributes_v2(Accumulate Dive=0.1, Accumulate Rest=-0.1, Bone Index=0, SkeletalMesh:NiagaraDataInterfaceSkeletalMesh=$User.Skeletal Mesh01)
   /SetVariables(Particles.Dive01:NiagaraFloat=&Multiply_Float(A=1, B=$Particles.Dive)) {isSetParameters=true}
   &ParticleState(Kill Particles When Lifetime Has Elapsed=false, Let Infinitely Lived Particles Die When Emitter Deactivates=true)
   &SkeletalMeshLocation(
    Apply Sampled Orientation to Mesh Particles=true
    Bone / Socket Index=$Particles.SpawnGroup
    Bone Sampling Mode=::NewEnumerator4
    Inherit Velocity=false
    MaskSpawnGroup=false
    Sampled Position Offset Coordinate Space:ENiagaraMeshTransforms=::NewEnumerator2
    Skeletal Mesh=$User.Skeletal Mesh02
   )
   &ScaleColor(
    Scale Alpha=&FNTimeOfDayScalar(Night Scale=0.5, Morning Scale=0.7, Evening Scale=0.7)
    Scale RGB=&VectorFromFloat(Value=&Lerp_Float(A=0.3, B=0.6, Alpha=$Particles.Dive01))
    ScaleA=true
    ScaleRGB=true
   ) {module=ScaleColor001}
  ]
  rendererClasses [&NiagaraLightRendererProperties]
  renderers [{rendererClass=&NiagaraLightRendererProperties, rendererIndex=0}]
 }
 Ribbon_Feet {
  emitterUpdateScript [
   &EmitterState(Life Cycle Mode=::NewEnumerator1, Loop Duration=0, UseLoopDelay=false)
   /Game/Effects/Niagara/Modules/Skydiving/FN_DivingAttributes/FN_EmitterDivingAttributes_v1(Bone Index=0, SkeletalMesh:NiagaraDataInterfaceSkeletalMesh=$User.SkeletalMesh)
   &FN_SpawnRibbon_v2(Number of Ribbons=2, Spawn Mode=::NewEnumerator0, SpawnRate=64)
  ]
  particleSpawnScript [
   &FN_SetSpawnGroup(Number of Spawn Groups=2)
   &FN_InitializeRibbon_v3(Boneflow Offset=2, Color:LinearColor=$User.Variant_Color, Enable Skeletal Position=true, Lifetime=0.2, Ribbon Width=15)
   &AddVelocity(
    Coordinate Space=::World
    Scale Added Velocity=&Lerp_Vector(A=<1 1 1>, B=<1 1 2>, Alpha=$Emitter.Dive)
    Velocity=&Lerp_Vector(A=<0 0 800>, B=<0 0 1200>, Alpha=$System.Dive)
   ) {module=AddVelocity003}
   &AddVelocity(
    Scale Added Velocity=&Lerp_Vector(A=<1 1 1>, B=<1 1 2>, Alpha=$Emitter.Dive)
    Velocity=&FN_FrontEndVector3Modifier(Gameplay=<0 200 0>, Front End=&Lerp_Vector(A=<0 -200 0>, B=<1 -500 1>, Alpha=$System.Dive))
   ) {module=AddVelocity002}
   &DynamicMaterialParameters(
    Index 0 Param 1=&MaskFloatBySpawnGroup(Masked Float=-0.5, Passthrough Float=-1, Spawn Group Mask=3)
    Index 0 Param 2=&FN_FrontEndFloatModifier(
     Gameplay=&Multiply_Float(A=$System.FN_Timer001.Value, B=1)
     Front End=&Multiply_Float(A=$System.FN_Timer001.Value, B=1)
    )
    Module.Param0WriteEnabled=true
    Module.Param1WriteEnabled=true
    Module.Param2WriteEnabled=false
    Module.Param3WriteEnabled=false
   )
   &SkeletalMeshLocation(
    Apply Sampled Orientation to Mesh Particles=true
    Bone / Socket Index=$Particles.SpawnGroup
    Bone Sampling Mode=::NewEnumerator4
    Inherit Velocity=false
    Interpolate Mesh Sampling=false
    MaskSpawnGroup=false
    Sampled Position Offset=&Lerp_Vector(A=<0 0 0>, B=<0 0 0>, Alpha=$System.Dive)
    Sampled Position Offset Coordinate Space:ENiagaraMeshTransforms=::NewEnumerator2
    Skeletal Mesh=$User.Skeletal Mesh02
   )
  ]
  particleUpdateScript [
   /Niagara/Noise/Waveform(
    EnableExponent1=false
    Global Amplitude Scale=10
    GlobalClampMax=false
    GlobalClampMin=false
    UseInput1=false
    UsePhase1=false
    "Waveform [1]":ENiagara_Waveforms=::NewEnumerator12
    "[1] Amplitude Min/Max":Vector2f=<-0.5 0.5>
    "[1] Frequency"=0.5
    "[1] Phase Scale"=2
    "[1] Scale"=2
   )
   /SetVariables(Particles.Velocity=&Add_Vector(A=$Particles.Velocity, B=&VectorFromFloat(Value=$Particles.Waveform.Waveform))) {isSetParameters=true, module=SetVariables_#11}
   &InheritSourceMovement(Applied Position Scale=<0.4 0.4 1>, ApplyForce=false, ApplyPosition=true, ApplyPositionFalloff=false)
   &SolveForcesAndVelocity(Clamp Velocity=false, Limit Acceleration=false)
   &FN_UpdateRibbon(Detach Locked Particle=true, Enable Skeletal Position=true)
   &ScaleColor(
    Linear Color Curve=DataInterface() {propertyValues {
      AlphaCurve {keys [<at 0 1>,<at 0.7457 0>,<at 1 0>]}
      BlueCurve {keys [<at 0 1>,<at 0.2227 0.6198>,<at 0.6742 0.4792>]}
      GreenCurve {keys [<at 0 1>,<at 0.2227 0.6198>,<at 0.6742 0.3394>]}
      RedCurve {keys [<at 0 1>,<at 0.2227 0.6198>,<at 0.6742 0.3145>]}
     }}
    Scale Mode=::NewEnumerator2
   )
  ]
  rendererClasses [&NiagaraRibbonRendererProperties]
  renderers [{rendererClass=&NiagaraRibbonRendererProperties}]
 }
 Ribbon_Left {
  emitterUpdateScript [
   &EmitterState(Life Cycle Mode=::NewEnumerator1, Loop Duration=0, UseLoopDelay=false)
   &FN_EmitterDivingAttributes_v1(Bone Index=0, SkeletalMesh:NiagaraDataInterfaceSkeletalMesh=$User.SkeletalMesh)
   &FN_SpawnRibbon_v2(Number of Ribbons=2, Spawn Mode=::NewEnumerator0, SpawnRate=64)
  ]
  particleSpawnScript [
   &FN_SetSpawnGroup(Number of Spawn Groups=2)
   &FN_InitializeRibbon_v3(
    Color:LinearColor=$User.Variant_Color
    Enable Skeletal Position=true
    Lifetime:NiagaraFloat=&FN_FrontEndFloatModifier(Gameplay=0.225, Front End=&Lerp_Float(A=0.225, B=0.17, Alpha=$System.Dive))
    Ribbon Width=12
   )
   &AddVelocity(
    Coordinate Space=::World
    Scale Added Velocity=&Lerp_Vector(A=<1 1 1>, B=<1 1 2>, Alpha=$Emitter.Dive)
    Velocity=&Lerp_Vector(A=<0 0 800>, B=<0 0 1600>, Alpha=$System.Dive)
   ) {module=AddVelocity003}
   &AddVelocity(
    Scale Added Velocity=&Lerp_Vector(A=<1 1 1>, B=<1 1 2>, Alpha=$Emitter.Dive)
    Velocity=&FN_FrontEndVector3Modifier(Gameplay=<0 200 0>, Front End=&Lerp_Vector(A=<0 0 0>, B=<1 -1200 1>, Alpha=$System.Dive))
   ) {module=AddVelocity002}
   &DynamicMaterialParameters(
    Index 0 Param 1=&MaskFloatBySpawnGroup(Masked Float=0.33, Passthrough Float=0.66)
    Index 0 Param 2=&FN_FrontEndFloatModifier(
     Gameplay=&Multiply_Float(A=$System.FN_Timer001.Value, B=0.95)
     Front End=&Multiply_Float(A=$System.FN_Timer001.Value, B=0.95)
    )
    Module.Param0WriteEnabled=true
    Module.Param1WriteEnabled=true
    Module.Param2WriteEnabled=false
    Module.Param3WriteEnabled=false
   )
   &SkeletalMeshLocation(
    Apply Sampled Orientation to Mesh Particles=true
    Bone / Socket Index=$Particles.SpawnGroup
    Bone Sampling Mode=::NewEnumerator4
    Inherit Velocity=false
    Interpolate Mesh Sampling=false
    MaskSpawnGroup=false
    Sampled Position Offset=&Lerp_Vector(
     A=&FN_FrontEndVector3Modifier(Gameplay=<-7 0 10>, Front End=<-8 0 0>)
     B=&FN_FrontEndVector3Modifier(Gameplay=<-7 0 0>, Front End=<-8 0 0>)
     Alpha=$System.Dive
    )
    Sampled Position Offset Coordinate Space:ENiagaraMeshTransforms=::NewEnumerator2
    Skeletal Mesh=$User.SkeletalMesh
   )
  ]
  particleUpdateScript [
   &Waveform(
    EnableExponent1=false
    Global Amplitude Scale=10
    GlobalClampMax=false
    GlobalClampMin=false
    UseInput1=false
    UsePhase1=false
    "Waveform [1]":ENiagara_Waveforms=::NewEnumerator12
    "[1] Amplitude Min/Max":Vector2f=<-0.5 0.5>
    "[1] Frequency"=0.5
    "[1] Phase Scale"=2
    "[1] Scale"=2
   )
   /SetVariables(Particles.Velocity=&Add_Vector(A=$Particles.Velocity, B=&VectorFromFloat(Value=$Particles.Waveform.Waveform))) {isSetParameters=true}
   &InheritSourceMovement(Applied Position Scale=<0.4 0.4 1>, ApplyForce=false, ApplyPosition=true, ApplyPositionFalloff=false)
   &SolveForcesAndVelocity(Clamp Velocity=false, Limit Acceleration=false)
   &FN_UpdateRibbon(Detach Locked Particle=true, Enable Skeletal Position=true)
   &ScaleColor(
    Linear Color Curve=DataInterface() {propertyValues {
      AlphaCurve {keys [<at 0 1>,<at 1 0>]}
      BlueCurve {keys [<at 0 1>,<at 0.2227 0.6198>,<at 0.6742 0.4792>]}
      GreenCurve {keys [<at 0 1>,<at 0.2227 0.6198>,<at 0.6742 0.3394>]}
      RedCurve {keys [<at 0 1>,<at 0.2227 0.6198>,<at 0.6742 0.3145>]}
     }}
    Scale Mode=::NewEnumerator2
   )
  ]
  rendererClasses [&NiagaraRibbonRendererProperties]
  renderers [{rendererClass=&NiagaraRibbonRendererProperties}]
 }
 Ribbon_Right {
  emitterUpdateScript [
   &EmitterState(Life Cycle Mode=::NewEnumerator1, Loop Duration=0, UseLoopDelay=false)
   &FN_EmitterDivingAttributes_v1(Bone Index=0, SkeletalMesh:NiagaraDataInterfaceSkeletalMesh=$User.SkeletalMesh)
   &FN_SpawnRibbon_v2(Number of Ribbons=2, Spawn Mode=::NewEnumerator0, SpawnRate=64)
  ]
  particleSpawnScript [
   &FN_SetSpawnGroup(Number of Spawn Groups=2)
   &FN_InitializeRibbon_v3(
    Color:LinearColor=$User.Variant_Color
    Enable Skeletal Position=true
    Lifetime:NiagaraFloat=&FN_FrontEndFloatModifier(Gameplay=0.225, Front End=&Lerp_Float(A=0.225, B=0.17, Alpha=$System.Dive))
    Ribbon Width=12
   )
   &AddVelocity(
    Coordinate Space=::World
    Scale Added Velocity=&Lerp_Vector(A=<1 1 1>, B=<1 1 2>, Alpha=$Emitter.Dive)
    Velocity=&Lerp_Vector(A=<0 0 800>, B=<0 0 1600>, Alpha=$System.Dive)
   ) {module=AddVelocity003}
   &AddVelocity(
    Scale Added Velocity=&Lerp_Vector(A=<1 1 1>, B=<1 1 2>, Alpha=$Emitter.Dive)
    Velocity=&FN_FrontEndVector3Modifier(Gameplay=<0 200 0>, Front End=&Lerp_Vector(A=<0 0 0>, B=<1 -1200 1>, Alpha=$System.Dive))
   ) {module=AddVelocity002}
   &DynamicMaterialParameters(
    Index 0 Param 1=&MaskFloatBySpawnGroup(Masked Float=-0.5, Passthrough Float=-1)
    Index 0 Param 2=&FN_FrontEndFloatModifier(
     Gameplay=&Multiply_Float(A=$System.FN_Timer001.Value, B=0.95)
     Front End=&Multiply_Float(A=$System.FN_Timer001.Value, B=0.95)
    )
    Module.Param0WriteEnabled=true
    Module.Param1WriteEnabled=true
    Module.Param2WriteEnabled=false
    Module.Param3WriteEnabled=false
   )
   &SkeletalMeshLocation(
    Apply Sampled Orientation to Mesh Particles=true
    Bone / Socket Index=$Particles.SpawnGroup
    Bone Sampling Mode=::NewEnumerator4
    Inherit Velocity=false
    Interpolate Mesh Sampling=false
    MaskSpawnGroup=false
    Sampled Position Offset=&Lerp_Vector(
     A=&FN_FrontEndVector3Modifier(Gameplay=<7 0 10>, Front End=<8 0 0>)
     B=&FN_FrontEndVector3Modifier(Gameplay=<7 0 0>, Front End=<10 0 0>)
     Alpha=$System.Dive
    )
    Sampled Position Offset Coordinate Space:ENiagaraMeshTransforms=::NewEnumerator2
    Skeletal Mesh=$User.SkeletalMesh
   )
  ]
  particleUpdateScript [
   &Waveform(
    EnableExponent1=false
    Global Amplitude Scale=10
    GlobalClampMax=false
    GlobalClampMin=false
    UseInput1=false
    UsePhase1=false
    "Waveform [1]":ENiagara_Waveforms=::NewEnumerator12
    "[1] Amplitude Min/Max":Vector2f=<-0.5 0.5>
    "[1] Frequency"=0.5
    "[1] Phase Scale"=2
    "[1] Scale"=2
   )
   /SetVariables(Particles.Velocity=&Add_Vector(A=$Particles.Velocity, B=&VectorFromFloat(Value=$Particles.Waveform.Waveform))) {isSetParameters=true, module=SetVariables_#11}
   &InheritSourceMovement(Applied Position Scale=<0.4 0.4 1>, ApplyForce=false, ApplyPosition=true, ApplyPositionFalloff=false)
   &SolveForcesAndVelocity(Clamp Velocity=false, Limit Acceleration=false)
   &FN_UpdateRibbon(Detach Locked Particle=true, Enable Skeletal Position=true)
   &ScaleColor(
    Linear Color Curve=DataInterface() {propertyValues {
      AlphaCurve {keys [<at 0 1>,<at 1 0>]}
      BlueCurve {keys [<at 0 1>,<at 0.2227 0.6198>,<at 0.6742 0.4792>]}
      GreenCurve {keys [<at 0 1>,<at 0.2227 0.6198>,<at 0.6742 0.3394>]}
      RedCurve {keys [<at 0 1>,<at 0.2227 0.6198>,<at 0.6742 0.3145>]}
     }}
    Scale Mode=::NewEnumerator2
   )
  ]
  rendererClasses [&NiagaraRibbonRendererProperties]
  renderers [{rendererClass=&NiagaraRibbonRendererProperties}]
 }
 Runes {
  emitterUpdateScript [
   &EmitterState()
   &SpawnRate(SpawnRate=20, Use Spawn Probability=false)
  ]
  particleSpawnScript [
   &FN_SetSpawnGroup(Number of Spawn Groups=2)
   &InitializeParticle(
    Color=&MultiplyLinearColor(Linear Color=$User.Variant_Color, Scale Factor=<3 3 3 1>)
    Lifetime=&RandomRangeFloat(Minimum=0.33, Maximum=0.66)
    Sprite Rotation Mode=::NewEnumerator2
    Sprite Size Mode=::NewEnumerator1
    Uniform Sprite Size Max=15
    Uniform Sprite Size Min=10
    UsePositionOffset=false
    Write Material Random=true
   )
   &SubUVAnimation(End Frame=3, SubUV Animation Mode:ENiagara_SubUVLookupMode=::NewEnumerator2)
   &SkeletalMeshLocation(
    Bone / Socket Index=$Particles.SpawnGroup
    Bone Sampling Mode=::NewEnumerator4
    Inherit Velocity=false
    MaskSpawnGroup=false
    Skeletal Mesh=$User.Skeletal Mesh02
   )
   &CameraOffset(Camera Offset Amount=&RandomRangeFloat(Minimum=10, Maximum=20))
   &AddVelocity(
    Coordinate Space=::World
    Velocity=&Lerp_Vector(
     A=<0 0 400>
     B=&FN_FrontEndVector3Modifier(Gameplay=<0 0 1600>, Front End=<0 0 0>)
     Alpha=$System.Dive
    )
   )
   &AddVelocity(Velocity=&FN_FrontEndVector3Modifier(
     Gameplay=&Lerp_Vector(A=<0 75 0>, B=<1 200 1>, Alpha=$System.Dive)
     Front End=&Lerp_Vector(A=<0 -10 0>, B=<1 -400 600>, Alpha=$System.Dive)
    )) {module=AddVelocity003}
  ]
  particleUpdateScript [
   &FN_DivingAttributes_v2(Accumulate Dive=0.1, Accumulate Rest=-0.1, Bone Index=0, SkeletalMesh:NiagaraDataInterfaceSkeletalMesh=$User.Skeletal Mesh01)
   &InheritSourceMovement(Applied Position Scale=<0.5 0.5 1>, ApplyForce=false, ApplyPosition=true, ApplyPositionFalloff=false)
   &ParticleState()
   &ScaleColor(
    Scale Alpha=&FloatFromCurve(FloatCurve=DataInterface() {propertyValues {Curve {keys [
         {arriveTangent=9.0725, interpMode=RCIM_Cubic, leaveTangent=9.0725, tangentMode=RCTM_User, time=0, value=0}
         {arriveTangent=0.444, interpMode=RCIM_Cubic, leaveTangent=0.444, tangentMode=RCTM_User, time=0.1408, value=0.9847}
         {arriveTangent=-0.0405, interpMode=RCIM_Cubic, leaveTangent=-0.0405, tangentMode=RCTM_User, time=1, value=0}
        ]}}})
    ScaleA=true
    ScaleRGB=true
   ) {module=ScaleColor001}
   &Drag(Drag=&Lerp_Float(A=2.5, B=5, Alpha=$System.Dive), Use Linear Drag=true, Use Rotational Drag=false)
   &SpriteRotationRate(Rotation Rate=&RandomRangeFloat(Minimum=180, Maximum=-180))
   &SolveForcesAndVelocity(Clamp Velocity=false, Limit Acceleration=false)
   &ScaleSpriteSize(Scale Factor=&Vector2DFromFloat(Value=&FloatFromCurve(FloatCurve=DataInterface() {propertyValues {Curve {keys [<at 0 0>,<at 0.1 1>]}}})))
  ]
  rendererClasses [&NiagaraSpriteRendererProperties]
  renderers [{rendererClass=&NiagaraSpriteRendererProperties}]
 }
 Sparks {
  emitterUpdateScript [
   &EmitterState()
   &SpawnRate(SpawnRate=24, Use Spawn Probability=false)
   /Game/Effects/Niagara/Modules/Skydiving/FN_Emitter_SkydiveAttributes() {module=Emitter_SkydiveAttributes}
  ]
  particleSpawnScript [
   &FN_SetSpawnGroup(Number of Spawn Groups=2)
   &InitializeParticle(
    Color=&MultiplyLinearColor(Linear Color=$User.Variant_Color, Scale Factor=<3 3 3 1>)
    Lifetime=0.75
    Sprite Size Max=<1.5 5>
    Sprite Size Min=<0.7 1>
    Sprite Size Mode=::NewEnumerator2
    UsePositionOffset=false
    Write Material Random=true
   ) {module=InitializeParticle001}
   &SkeletalMeshLocation(
    Bone / Socket Index=$Particles.SpawnGroup
    Bone Sampling Mode=::NewEnumerator4
    Inherit Velocity=false
    Interpolate Mesh Sampling=false
    MaskSpawnGroup=false
    Sampled Position Offset=<0 10 0>
    Sampled Position Offset Coordinate Space:ENiagaraMeshTransforms=::NewEnumerator2
    Skeletal Mesh=$User.Skeletal Mesh02
   )
   &SphereLocation(MaskSpawnGroup=false, Module.Override Randomness=false, Sphere Radius=0, Surface Only=false)
   &AddVelocityInCone(
    Cone Angle=30
    Cone Axis=&Lerp_Vector(
     A=<0 0 1>
     B=&FN_FrontEndVector3Modifier(Gameplay=<0 -0.2 1>, Front End=<0 -1 -0.5>)
     Alpha=$System.Dive
    )
    Module.Use Velocity Falloff On Cone Axis=false
    Velocity Strength=&FN_FrontEndFloatModifier(Gameplay=&Lerp_Float(A=400, B=800, Alpha=$System.Dive), Front End=100)
   )
  ]
  particleUpdateScript [
   &FN_DivingAttributes_v2(Accumulate Dive=0.1, Accumulate Rest=-0.1, Bone Index=0, SkeletalMesh:NiagaraDataInterfaceSkeletalMesh=$User.Skeletal Mesh01)
   /SetVariables(Particles.Dive01:NiagaraFloat=&Multiply_Float(A=1, B=$Particles.Dive)) {isSetParameters=true}
   &ParticleState()
   &AddVelocity(Velocity=&Lerp_Vector(A=<0 0 10>, B=<0 -10 20>, Alpha=$System.Dive))
   &VectorNoiseForce(Force Amount=1500, Module.Use Random Noise Time Interval=true, Random Noise Time Interval=0.001)
   &Drag(Drag=&RandomRangeFloat(Minimum=2, Maximum=4), Use Linear Drag=true, Use Rotational Drag=false)
   &GravityForce(Gravity=&Multiply_Vector(A=<0 0 250>, B=&Lerp_Vector(A=<0 0 0>, B=<0 0 1>, Alpha=$Particles.Dive01)))
   &SolveForcesAndVelocity(Clamp Velocity=false, Limit Acceleration=false)
   /SetVariables(Particles.MultColor:LinearColor=&MakeLinearColorFromVectorAndFloat(
     "Vector (RGB)"=&Multiply_Vector(
      A=&MakeVectorFromLinearColorRGB(Color=$Particles.Color)
      B=&MakeVectorFromLinearColorRGB(Color=&ColorFromCurve(ColorCurve=DataInterface() {propertyValues {
          AlphaCurve {keys [<at 0 0 RCIM_Cubic>,<at 0.1151 1>,<at 0.9977 1 RCIM_Cubic>]}
          BlueCurve {keys [<at 0.0017 2>,<at 0.9821 2>]}
          GreenCurve {keys [<at 0.0017 2>,<at 0.9821 2>]}
          RedCurve {keys [<at 0.0017 2>,<at 0.9821 2>]}
         }}))
     )
     "Float (Alpha)"=1
    )) {isSetParameters=true}
   &ScaleColor(
    Scale Alpha=&Multiply_Float(A=&FNTimeOfDayScalar(Night Scale=0.5, Morning Scale=0.7, Evening Scale=0.7), B=1)
    Scale RGB=<2 2 2>
    ScaleA=true
    ScaleRGB=true
   )
   &ScaleSpriteSizeBySpeed(Max Scale Factor=<1 3>, Min Scale Factor=<1 1.05>)
   &ScaleSpriteSize(Scale Factor=&Vector2DFromCurve(
     Vector2Curve=DataInterface() {propertyValues {
       XCurve {keys [
         <at 0 0 RCIM_Cubic>
         {arriveTangent=-1.6411, interpMode=RCIM_Cubic, leaveTangent=-1.6411, tangentMode=RCTM_User, time=0.15, value=1}
         <at 1 0>
        ]}
       YCurve {keys [<at 0 0 RCIM_Cubic>,<at 0.15 1 RCIM_Cubic>,<at 1 0>]}
      }}
     ScaleCurve=<1 1>
    ))
  ]
  rendererClasses [&NiagaraSpriteRendererProperties]
  renderers [{rendererClass=&NiagaraSpriteRendererProperties}]
 }
}
summary {
 systemName=NS_Contrail_Lancelot
 emitters [
  {emitterName=Sparks, rendererClasses [&NiagaraSpriteRendererProperties]}
  {emitterName=Light, rendererClasses [&NiagaraLightRendererProperties]}
  {emitterName=Runes, rendererClasses [&NiagaraSpriteRendererProperties]}
  {emitterName=Ribbon_Left, rendererClasses [&NiagaraRibbonRendererProperties]}
  {emitterName=Fire_Mesh_H, rendererClasses [&NiagaraSpriteRendererProperties, &NiagaraMeshRendererProperties]}
  {emitterName=Bottom_Flame_H, rendererClasses [&NiagaraSpriteRendererProperties]}
  {emitterName=Front_Flame_H, rendererClasses [&NiagaraSpriteRendererProperties]}
  {emitterName=Ribbon_Right, rendererClasses [&NiagaraRibbonRendererProperties]}
  {emitterName=Decal_Flame_H, rendererClasses [&NiagaraSpriteRendererProperties]}
  {emitterName=Bottom_Flame, rendererClasses [&NiagaraSpriteRendererProperties]}
  {emitterName=Front_Flame, rendererClasses [&NiagaraSpriteRendererProperties]}
  {emitterName=Decal_Flame, rendererClasses [&NiagaraSpriteRendererProperties]}
  {emitterName=FloatingBits, rendererClasses [&NiagaraMeshRendererProperties]}
  {emitterName=Bits, rendererClasses [&NiagaraMeshRendererProperties]}
  {bEnabled=false, emitterName=FN_Debug, rendererClasses [&NiagaraSpriteRendererProperties]}
  {emitterName=Ribbon_Feet, rendererClasses [&NiagaraRibbonRendererProperties]}
  {emitterName=Fire_Mesh, rendererClasses [&NiagaraSpriteRendererProperties, &NiagaraMeshRendererProperties]}
 ]
 userVariables [
  User.bIsFrontEnd {type=NiagaraBool}
  User.bIsFrontEndPreview {type=NiagaraBool}
  User.bIsFullyDeployed {type=NiagaraBool}
  User.ForwardDot {type=NiagaraFloat, defaultValue=0, description=""}
  User.IsSkydiving {type=NiagaraBool}
  User.ParaGlideLeanAlpha {type=NiagaraFloat, defaultValue=0, description=""}
  User.PlayerSpeed {type=NiagaraFloat, defaultValue=0, description=""}
  User.RibbonIDOffset {type=NiagaraInt32}
  User.RightDot {type=NiagaraFloat, defaultValue=0, description=""}
  User.Skeletal Mesh01 {
   type=NiagaraDataInterfaceSkeletalMesh
   defaultValue {
    struct=NiagaraExt_VariableValue_DataInterface
    value {dataInterface=&self:NiagaraDataInterfaceSkeletalMesh_2}
   }
  }
  User.Skeletal Mesh02 {
   type=NiagaraDataInterfaceSkeletalMesh
   defaultValue {
    struct=NiagaraExt_VariableValue_DataInterface
    value {dataInterface=&self:NiagaraDataInterfaceSkeletalMesh_0}
   }
  }
  User.SkeletalMesh {
   type=NiagaraDataInterfaceSkeletalMesh
   defaultValue {
    struct=NiagaraExt_VariableValue_DataInterface
    value {dataInterface=&self:NiagaraDataInterfaceSkeletalMesh_3}
   }
  }
  User.Variant_Color {type=LinearColor, description="", defaultValue {struct=LinearColor, value=<rgba 1 1 1 1>}}
 ]
}
